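# Niagara system for the Jump Boots item that flashes a brief emissive glow on two bones or sockets of a skeletal mesh.
# The Glow emitter bursts short-lived sprites in two spawn groups, each mapped to a bone index.
# A normalized height value computed from spawn index spaces particles along an offset and blends their color and size.
# Color curves fade the bright flash, and a height comparison gates particle lights; the Empty emitter is disabled.
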NiagaraSystem /PrimalGameplay/Items/JumpBoots/Assets/Effects/NS_JumpBoots_Emissive
dependencies {
 usedDataInterfaces [&NiagaraDataInterfaceColorCurve]
 usedDynamicInputs []
 usedModules [
  &self:SystemSpawnScript.SystemScriptSource.SystemScriptGraph.NiagaraNodeAssignment_0.SetVariables_#1
  &SystemState
  &self:SystemSpawnScript.SystemScriptSource.SystemScriptGraph.NiagaraNodeAssignment_1.SetVariables_#2
 ]
}
emitters {
 Empty {
  bEnabled=false
  emitterUpdateScript [
   &EmitterState()
   &SpawnBurst_Instantaneous(Spawn Count=1, Use Spawn Probability=false)
  ]
  particleSpawnScript [&InitializeParticle(UsePositionOffset=false, Write Material Random=true)]
  particleUpdateScript [&ParticleState()]
  rendererClasses [&NiagaraSpriteRendererProperties]
  renderers [{rendererClass=&NiagaraSpriteRendererProperties}]
 }
 Glow {
  emitterUpdateScript [
   &EmitterState()
   /SetVariables(Emitter.BurstSpawnCount=16) {isSetParameters=true}
   &SpawnBurst_Instantaneous(Spawn Count=$Emitter.BurstSpawnCount, Use Spawn Probability=false)
  ]
  particleSpawnScript [
   /SetVariables(
    Transient.LessThanHalfSpawn:NiagaraFloat=&Add_Float(
     A=&Multiply_Float(A=&MakeFloatFromInt(Int=$Emitter.BurstSpawnCount), B=0.5)
     B=-0.01
    )
    Transient.NormalizePosMult:NiagaraFloat=&Divide_Float(
     A=1
     B=&Add_Float(
      A=&Multiply_Float(A=&MakeFloatFromInt(Int=$Emitter.BurstSpawnCount), B=0.5)
      B=-1
     )
    )
   ) {isSetParameters=true}
   /SetVariables(
    Particles.HeightPos:NiagaraFloat=&Multiply_Float(
     A=$Transient.NormalizePosMult
     B=/Niagara/DynamicInputs/Math/Floor_Float(Float:NiagaraFloat=&Multiply_Float(A=&ReturnNormalizedExecIndex(), B=$Transient.LessThanHalfSpawn))
    )
    Particles.PickBone:NiagaraInt32=&ModuloInt(
     A=&MakeIntFromFloat(Float=&Multiply_Float(
       A=&ReturnNormalizedExecIndex(Normalized Index Includes One=false)
       B=&MakeFloatFromInt(Int=$Emitter.BurstSpawnCount)
      ))
     B=2
    )
   ) {isSetParameters=true}
   &InitializeParticle(
    Color=&LerpLinearColors(StartColor=<rgba 0.0667 0 1 0.05>, EndColor=<rgba 0 2 0.6333 0.1>, LerpFactor=$Particles.HeightPos)
    Lifetime Max=0.25
    Lifetime Min=0.1
    Lifetime Mode=::NewEnumerator1
    Sprite Size Mode=::NewEnumerator3
    Uniform Sprite Size=&Lerp_Float(
     A=10
     B=20
     Alpha=&ClampFloat(Float=&Add_Float(A=$Particles.HeightPos, B=&RandomRangeFloat(Minimum=-0.25, Maximum=0.25)))
    )
    UsePositionOffset=false
    Write Material Random=true
   )
   &FN_SetSpawnGroup(Number of Spawn Groups=2)
  ]
  particleUpdateScript [
   &ParticleState()
   &SkeletalMeshLocation(
    Bone / Socket Index=&SpawnGroup()
    Bone Sampling Mode=::NewEnumerator4
    Inherit Velocity=false
    MaskSpawnGroup=false
    Sampled Position Offset=&Lerp_Vector(A=<0 0 0>, B=<15 0 2>, Alpha=$Particles.HeightPos)
    Sampled Position Offset Coordinate Space:ENiagaraMeshTransforms=::NewEnumerator3
    Skeletal Mesh=$User.Skeletal Mesh
   )
   &ScaleColor(
    Linear Color Curve=DataInterface() {propertyValues {
      AlphaCurve {keys [
        {arriveTangent=-13.3965, interpMode=RCIM_Cubic, leaveTangent=-13.3965, tangentMode=RCTM_User, time=0, value=1}
        {arriveTangent=-1.1612, interpMode=RCIM_Cubic, leaveTangent=-1.1612, tangentMode=RCTM_User, time=0.074, value=0.5}
        {arriveTangent=-0.5784, interpMode=RCIM_Cubic, leaveTangent=-0.5784, tangentMode=RCTM_User, time=1, value=0}
       ]}
      BlueCurve {keys [
        {arriveTangent=-17.8945, interpMode=RCIM_Cubic, time=0, value=10}
        {arriveTangent=-8, interpMode=RCIM_Cubic, leaveTangent=-8, time=0.162, value=9.4383}
        <at 1 2>
       ]}
      GreenCurve {keys [
        {arriveTangent=-11.6314, interpMode=RCIM_Cubic, leaveTangent=-10.3904, tangentMode=RCTM_User, time=0, value=10}
        {arriveTangent=-13.2448, interpMode=RCIM_Cubic, leaveTangent=-13.2448, tangentMode=RCTM_User, time=0.1624, value=3.9077}
        <at 1 0.248>
       ]}
      RedCurve {keys [
        {arriveTangent=-8.7522, interpMode=RCIM_Cubic, leaveTangent=-5.8515, tangentMode=RCTM_User, time=0, value=10}
        {arriveTangent=-13.9975, interpMode=RCIM_Cubic, leaveTangent=-13.9975, tangentMode=RCTM_User, time=0.1624, value=3.612}
        <at 1 0.9196>
       ]}
     }}
    Scale Mode=::NewEnumerator2
   )
   &FN_Light_Attributes(
    Color=$Particles.Color
    Create Color Attribute=true
    Create Exponent Attribute=true
    Create Radius Attribute=true
    Enable Particle Light Enable=true
    Exponent=3
    Light Enabled:NiagaraBool=&SetBoolByFloatComparison(A=$Particles.HeightPos, B=1, Comparison Type=::NewEnumerator4)
    Radius=20
    Scale Color=&ClampFloat(Float=&Lerp_Float(A=2, B=0, Alpha=$Particles.NormalizedAge), Max=0.25)
   )
  ]
  rendererClasses [&NiagaraSpriteRendererProperties, &NiagaraLightRendererProperties]
  renderers [
   {rendererClass=&NiagaraSpriteRendererProperties}
   {rendererClass=&NiagaraLightRendererProperties, rendererIndex=1}
  ]
 }
}
summary {
 systemName=NS_JumpBoots_Emissive
 emitters [
  {bEnabled=false, emitterName=Empty, rendererClasses [&NiagaraSpriteRendererProperties]}
  {emitterName=Glow, rendererClasses [&NiagaraSpriteRendererProperties, &NiagaraLightRendererProperties]}
 ]
 userVariables [
  User.Delay {type=NiagaraFloat, defaultValue=0.2, description=""}
  User.Skeletal Mesh {
   type=NiagaraDataInterfaceSkeletalMesh
   defaultValue {
    struct=NiagaraExt_VariableValue_DataInterface
    value {dataInterface=&self:NiagaraDataInterfaceSkeletalMesh_0}
   }
  }
  User.StageNum {type=NiagaraInt32, defaultValue=1}
 ]
}
