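# Niagara effect for a Jump Boots jump: a one-shot burst of glowing ashes, a mesh beam and smoke.
# User StageNum and System.NormalizedStageNum scale intensity by stage, driving ash count, beam size, smoke size, speed, opacity and emissive glow.
# GPU ashes swirl with curl noise, while burst and lingering smoke fold velocity upward and animate erosion through a local material-dynamics module.
# An expanding upward-facing sprite adds the ground layer.
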NiagaraSystem /PrimalGameplay/Items/JumpBoots/Assets/Effects/NS_JumpBoots_OnJump
dependencies {
 usedDataInterfaces [
  &NiagaraDataInterfaceCurve
  &NiagaraDataInterfaceColorCurve
  &NiagaraDataInterfaceArrayFloat
  &NiagaraDataInterfaceArrayFloat3
  &NiagaraDataInterfaceVectorCurve
 ]
 usedDynamicInputs [&MakeFloatFromInt]
 usedModules [
  &self:SystemSpawnScript.SystemScriptSource.SystemScriptGraph.NiagaraNodeAssignment_0.SetVariables_#1
  &SystemState
 ]
}
emitters {
 Ashes {
  simTarget=GPUComputeSim
  emitterUpdateScript [
   &EmitterState(Life Cycle Mode=::NewEnumerator1, Loop Behavior=::NewEnumerator1, Loop Duration=0.25, UseLoopDelay=false)
   &SpawnBurst_Instantaneous(
    Spawn Count=&MakeIntFromFloat(Float=&Lerp_Float(A=0, B=15, Alpha=$System.NormalizedStageNum))
    Use Spawn Probability=false
   )
  ]
  particleSpawnScript [
   /SetVariables(
    Transient.BaseLifetime:NiagaraFloat=&UniformRangedFloat(Minimum=0.2, Maximum=&Lerp_Float(A=0.3, B=1, Alpha=$System.NormalizedStageNum))
    Transient.LongLifetime:NiagaraFloat=&Lerp_Float(
     A=&RandomRangeFloat(Minimum=0.5, Maximum=1)
     B=&RandomRangeFloat(Minimum=1.75, Maximum=2.5)
     Alpha=$System.NormalizedStageNum
    )
    Transient.PickLifetime:NiagaraBool=&SetBoolByFloatComparison(A=&RandomRangeFloat(Minimum=0, Maximum=3), B=2.5)
   ) {isSetParameters=true}
   /Niagara/Modules/Spawn/Initialization/InitializeParticle(
    Color=&LerpLinearColors(
     StartColor=<rgba 0.2588 0.1922 1 1>
     EndColor=&RandomRangeLinearColor(Minimum=<rgba 0 1 0.7667 1>, Maximum=<rgba 0.55 0 1 1>)
     LerpFactor=&UniformAOrBFloat(A=0, B=&UniformAOrBFloat(A=0))
    )
    Lifetime=&MakeCustomFloatFromBool(Bool=$Transient.PickLifetime, True Float=$Transient.LongLifetime, False Float=$Transient.BaseLifetime)
    Module.Write Color=true
    Module.Write Lifetime=true
    Module.Write Mass=true
    Module.Write Material Random=true
    Module.Write Position=false
    Module.Write Scale=false
    Module.Write SpriteRotation=true
    Module.Write SpriteSize=true
    Position=&SimulationPosition()
    Sprite Rotation=&UniformRangedFloat(Minimum=0, Maximum=360)
    Sprite Size=&Vector2DFromFloat(Value=&RandomRangeFloat(Minimum=1, Maximum=5))
   )
   &SphereLocation(
    Hemisphere Z=true
    MaskSpawnGroup=false
    Override Seed=false
    Sphere Radius=50
    Surface Only=false
    UseNonUniformScale=false
    UseOffset=true
    UseSphereOrientation=false
   )
   &AddVelocity(Velocity=&RandomRangeVector(Minimum=<-50 -50 -50>, Maximum=<50 50 50>))
   &AddVelocityFromPoint(Module.Use Falloff Distance=false, Velocity Strength=&RandomRangeFloat(Minimum=25, Maximum=200))
  ]
  particleUpdateScript [
   &ParticleState()
   &AccelerationForce(Acceleration=&RandomRangeVector(Minimum=<0 0 -200>, Maximum=<0 0 -100>))
   &Drag(Drag=0.75, Use Linear Drag=true, Use Rotational Drag=false)
   &CurlNoiseForce(Noise Frequency=10, Noise Strength=&Lerp_Float(A=500, B=0, Alpha=$Particles.NormalizedAge), Pan Noise=false)
   &SolveForcesAndVelocity(Clamp Velocity=false, Limit Acceleration=false)
   &ScaleColor(
    Scale RGB=&VectorFromFloat(Value=&Lerp_Float(A=6000, B=0, Alpha=$Particles.NormalizedAge))
    ScaleA=false
    ScaleRGB=true
   )
   &SpriteRotationRate(Rotation Rate=&RandomRangeFloat(Minimum=-2, Maximum=2))
   &ScaleSpriteSize(Scale Factor=&Vector2DFromFloat(Value=&FloatFromCurve(FloatCurve=DataInterface() {propertyValues {Curve {keys [
          {interpMode=RCIM_Cubic, tangentMode=RCTM_User, time=0, value=1}
          {arriveTangent=-0.5884, interpMode=RCIM_Cubic, leaveTangent=-0.5884, tangentMode=RCTM_User, time=0.758, value=0.8557}
          {arriveTangent=-5.513, interpMode=RCIM_Cubic, leaveTangent=-5.5131, tangentMode=RCTM_User, time=1, value=0}
         ]}}})))
  ]
  rendererClasses [&NiagaraSpriteRendererProperties]
  renderers [{rendererClass=&NiagaraSpriteRendererProperties}]
 }
 Beam {
  emitterUpdateScript [
   &EmitterState(
    Enable Distance Culling=true
    Inactive Response=::NewEnumerator1
    Life Cycle Mode=::NewEnumerator1
    Loop Behavior=::NewEnumerator1
    Loop Delay=0.075
    Loop Duration Mode=::NewEnumerator1
    Max Distance Response=::SleepAndClearParticles
    MaxDistance=7000
    Module.bUseMaxDistance=true
    Module.bUseMinDistance=false
    Reset Age On Awaken=true
    Scalability Mode=::NewEnumerator1
    UseLoopDelay=true
   )
   &SpawnBurst_Instantaneous(Spawn Count=1, Use Spawn Probability=false)
  ]
  particleSpawnScript [
   &InitializeParticle(
    Color=<rgba 0 0.2602 2 0.5>
    Lifetime=0.2
    Mass=&RandomRangeFloat(Minimum=0.6, Maximum=1.2)
    Mesh Scale=&SelectVectorFromArray(Direct Array Index=$User.StageNum, Vector Selection Array=DataInterface() {propertyValues {FloatData [<0 0 0>, <2.25 1 1>, <4 1 1>]}})
    Module.Write Color=true
    Module.Write Lifetime=true
    Module.Write Mass=false
    Module.Write Material Random=true
    Module.Write Position=false
    Module.Write Scale=true
    Module.Write SpriteRotation=true
    Module.Write SpriteSize=true
    Sprite Rotation=&NormalizedAngleToDegrees(NormalizedAngle=&RandomRangeFloat(Minimum=0, Maximum=1))
    Sprite Size=&Vector2DFromFloat(Value=&RandomRangeFloat(Minimum=400, Maximum=500))
   )
   &DynamicMaterialParameters(
    Index 0 Param 1=150
    Index 0 Param 3=&RandomRangeFloat(Minimum=0.4, Maximum=0.6)
    Index 0 Param 4=1
    Module.Param0WriteEnabled=true
    Module.Param1WriteEnabled=false
    Module.Param2WriteEnabled=false
    Module.Param3WriteEnabled=false
   )
   /Niagara/Modules/Spawn/Orientation/V2/InitialMeshOrientation(Orientation Vector=<0 0 1>, Use Orientation Vector=true, Use Rotation Vector=false) {enabled=false}
   /Niagara/Modules/Spawn/Orientation/V3/InitialMeshOrientation(Orientation Vector=<0 0 1>, Use Orientation Vector=true, Use Rotation Vector=false) {module=InitialMeshOrientation001}
  ]
  particleUpdateScript [
   &SolveForcesAndVelocity(Clamp Velocity=false, Limit Acceleration=false)
   &ParticleState(Kill Particles When Lifetime Has Elapsed=false, Let Infinitely Lived Particles Die When Emitter Deactivates=true)
   &ScaleColor(
    Scale Alpha=&FloatFromCurve(FloatCurve=DataInterface() {propertyValues {Curve {keys [
         {interpMode=RCIM_Cubic, tangentMode=RCTM_Break, time=0, value=0}
         {interpMode=RCIM_Cubic, tangentMode=RCTM_Break, time=0.2213, value=1}
         {arriveTangent=-0.2013, interpMode=RCIM_Cubic, tangentMode=RCTM_Break, time=1, value=0}
        ]}}})
    ScaleA=true
    ScaleRGB=true
   )
   &ScaleMeshSize(Scale Factor=&VectorFromCurve(
     VectorCurve=DataInterface() {propertyValues {XCurve {keys [<at 0 1>,<at 1 0>]}, YCurve {keys [<at 0 1>,<at 1 0>]}, ZCurve {keys [<at 0 1>,<at 1 0>]}}}
     Scale Curve=<1.5 1 1>
    ))
  ]
  rendererClasses [&NiagaraMeshRendererProperties]
  renderers [{rendererClass=&NiagaraMeshRendererProperties, rendererIndex=0}]
 }
 Empty {
  bEnabled=false
  emitterUpdateScript [
   &EmitterState()
   &SpawnBurst_Instantaneous(Spawn Count=1, Use Spawn Probability=false)
  ]
  particleSpawnScript [/Niagara/Modules/Spawn/Initialization/V2/InitializeParticle(
    Color=&LerpLinearColors(StartColor=<rgba 1 0 0 1>, EndColor=<rgba 0 1 0 1>, LerpFactor=$System.NormalizedStageNum)
    UsePositionOffset=false
    Write Material Random=true
   )]
  particleUpdateScript [&ParticleState()]
  rendererClasses [&NiagaraSpriteRendererProperties]
  renderers [{rendererClass=&NiagaraSpriteRendererProperties}]
 }
 SmokeBurst {
  emitterUpdateScript [
   &EmitterState()
   &SpawnBurst_Instantaneous(Spawn Count=30, Use Spawn Probability=false)
  ]
  particleSpawnScript [
   /SetVariables(
    Particles.InitSpriteRotRate=&UniformRangedFloat(Minimum=-3, Maximum=3)
    Transient.ScalePicker:NiagaraFloat=&UniformRangedFloat(Minimum=0, Maximum=1)
   ) {isSetParameters=true, module=SetVariables_#2}
   /Niagara/Modules/Spawn/Initialization/InitializeParticle(
    Color=<rgba 0.5 0.5 0.5 1>
    Lifetime=&Lerp_Float(
     A=&Lerp_Float(A=0.25, B=0.4, Alpha=$Transient.ScalePicker)
     B=&Lerp_Float(A=0.1, B=0.4, Alpha=$Transient.ScalePicker)
     Alpha=$System.NormalizedStageNum
    )
    Module.Write Color=true
    Module.Write Lifetime=true
    Module.Write Mass=true
    Module.Write Material Random=true
    Module.Write Position=true
    Module.Write Scale=false
    Module.Write SpriteRotation=true
    Module.Write SpriteSize=true
    Sprite Rotation=&RandomRangeFloat(Minimum=-360, Maximum=360)
    Sprite Size=&Vector2DFromFloat(Value=&Multiply_Float(
      A=&Lerp_Float(
       A=75
       B=200
       Alpha=&ClampFloat(Float=&Add_Float(A=$Transient.ScalePicker, B=&RandomRangeFloat(Minimum=-0.2, Maximum=0.3)))
      )
      B=&Lerp_Float(A=0.6, B=1, Alpha=$System.NormalizedStageNum)
     ))
   )
   &SphereLocation(
    MaskSpawnGroup=false
    Non Uniform Scale=<1 1 0.75>
    Override Seed=true
    Random Seed=20
    Sphere Coordinate Space=::World
    Sphere Radius=&Lerp_Float(A=10, B=50, Alpha=&Lerp_Float(A=0.5, B=1, Alpha=$System.NormalizedStageNum))
    Surface Only=false
    UseNonUniformScale=true
    UseOffset=false
    UseSphereOrientation=false
   )
   &AddVelocityFromPoint(
    Module.Use Falloff Distance=false
    Origin Offset=<0 0 -2>
    Origin Offset Coordinate Space=::World
    Velocity Strength=&Multiply_Float(
     A=&RandomRangeFloat(Minimum=100, Maximum=300)
     B=&Lerp_Float(A=0.7, B=1, Alpha=$System.NormalizedStageNum)
    )
   )
   /SetVariables(Particles.Velocity=&Multiply_Vector(
     A=&MakeVector(
      X=&MakeFloatFromVector(Vector=$Particles.Velocity, Channel=::NewEnumerator0)
      Y=&MakeFloatFromVector(Vector=$Particles.Velocity, Channel=::NewEnumerator1)
      Z=&Abs_Float(Float:NiagaraFloat=&MakeFloatFromVector(Vector=$Particles.Velocity, Channel=::NewEnumerator2))
     )
     B=<1 1 2>
    )) {isSetParameters=true, module=SetVariables_#3}
  ]
  particleUpdateScript [
   &ParticleState()
   &SpriteRotationRate(Rotation Rate=&Multiply_Float(
     A=$Particles.InitSpriteRotRate
     B=&FloatFromCurve(
      FloatCurve=DataInterface() {propertyValues {Curve {keys [
          {arriveTangent=-4.0107, interpMode=RCIM_Cubic, leaveTangent=-4.0107, tangentMode=RCTM_User, time=-0.0046, value=1.0113}
          {arriveTangent=-0.2496, interpMode=RCIM_Cubic, leaveTangent=-0.2496, tangentMode=RCTM_User, time=0.5, value=0.16}
         ]}}}
      CurveIndex=$Emitter.Age
      Scale Curve=150
     )
    ))
   &ScaleColor(
    Scale Alpha=&FloatFromCurve(
     FloatCurve=DataInterface() {propertyValues {Curve {keys [
         {arriveTangent=0.1469, interpMode=RCIM_Cubic, leaveTangent=0.1469, tangentMode=RCTM_User, time=0.1213, value=0.984}
         {arriveTangent=-0.6758, interpMode=RCIM_Cubic, leaveTangent=-0.6758, tangentMode=RCTM_User, time=0.3002, value=0.1865}
         <at 1 0>
        ]}}}
     Scale Curve=&Lerp_Float(A=0.5, B=1, Alpha=$System.NormalizedStageNum)
    )
    Scale RGB=&MakeVectorFromLinearColorRGB(Color=&ColorFromCurve(ColorCurve=DataInterface() {propertyValues {
        AlphaCurve {keys [<at 0 2 RCIM_Cubic>]}
        BlueCurve {keys [<at 0 0.3385 RCIM_Cubic>, {arriveTangent=-0.131, interpMode=RCIM_Cubic, leaveTangent=-0.131, time=0.2245, value=0.5}, <at 1.0334 0.2031>]}
        GreenCurve {keys [<at 0 0.2204 RCIM_Cubic>, {arriveTangent=-0.0167, interpMode=RCIM_Cubic, leaveTangent=-0.0167, time=0.2245, value=0.375}, <at 1.0334 0.2031>]}
        RedCurve {keys [<at 0 0.269 RCIM_Cubic>, {arriveTangent=-0.0637, interpMode=RCIM_Cubic, leaveTangent=-0.0637, time=0.2245, value=0.3958}, <at 1.0334 0.2031>]}
       }}))
    ScaleA=true
    ScaleRGB=true
   )
   /Niagara/Modules/Update/Forces/CurlNoiseForce(Noise Frequency=0.1, Noise Strength=800, Pan Noise=true, Pan Noise Field=<0 0 1>)
   &Drag(Drag=2, Use Linear Drag=true, Use Rotational Drag=false)
   &SolveForcesAndVelocity(Clamp Velocity=false, Limit Acceleration=false)
   &ScaleSpriteSize(Scale Factor=&Vector2DFromFloat(Value=&FloatFromCurve(FloatCurve=DataInterface() {propertyValues {Curve {keys [
          {arriveTangent=2.8475, interpMode=RCIM_Cubic, leaveTangent=2.3251, tangentMode=RCTM_User, time=0, value=0.15}
          {arriveTangent=1.1483, interpMode=RCIM_Cubic, leaveTangent=1.1483, tangentMode=RCTM_User, time=0.2651, value=0.6514}
          {arriveTangent=0.2754, interpMode=RCIM_Cubic, leaveTangent=0.2754, tangentMode=RCTM_User, time=1.0463, value=1}
         ]}}})))
   &self:Smoke_V2_Material_Dynamics(
    "Apply Noise (DissolveLerpNoise)"=1
    "Apply Noise (LerpNoise)"=1
    "Erode Alpha (DissolveNoiseErosionPower)":NiagaraFloat=&RandomRangeFloat(Minimum=0.5, Maximum=1)
    "Erode Emissive (NoiseErosionPower)":NiagaraFloat=&FloatFromCurve(FloatCurve=DataInterface() {propertyValues {Curve {keys [
         {arriveTangent=2.8986, interpMode=RCIM_Cubic, leaveTangent=2.8986, tangentMode=RCTM_User, time=0, value=0}
         {arriveTangent=5.2006, interpMode=RCIM_Cubic, leaveTangent=5.2006, tangentMode=RCTM_User, time=1, value=2}
        ]}}})
    "Intensity Modifier (Emissive)":NiagaraFloat=&FloatFromCurve(
     FloatCurve=DataInterface() {propertyValues {Curve {keys [{interpMode=RCIM_Cubic, tangentMode=RCTM_User, time=0, value=1}, <at 1 0>]}}}
     Scale Curve=&SelectFloatFromArray(Direct Array Index=$User.StageNum, Float Selection Array=DataInterface() {propertyValues {FloatData [0.025,0.1,0.5]}})
    )
    "Noise Offset X (DissolveNoiseOffsetX)":NiagaraFloat=&RandomRangeFloat(Minimum=0, Maximum=1)
    "Noise Offset X (NoiseOffsetX)":NiagaraFloat=&Add_Float(
     A=&RandomRangeFloat(Minimum=0, Maximum=1)
     B=&FloatFromCurve(
      FloatCurve=DataInterface() {propertyValues {Curve {keys [
          {arriveTangent=1.5771, interpMode=RCIM_Cubic, leaveTangent=1.223, tangentMode=RCTM_User, time=0, value=0}
          {arriveTangent=0.5984, interpMode=RCIM_Cubic, leaveTangent=0.5984, tangentMode=RCTM_User, time=0.321, value=0.3333}
          {arriveTangent=0.0529, interpMode=RCIM_Cubic, leaveTangent=0.2021, tangentMode=RCTM_User, time=1, value=0.5}
         ]}}}
      Scale Curve=&RandomRangeFloat(Minimum=0.25, Maximum=1)
     )
    )
    "Normalized Frame Number (Animation)":NiagaraFloat=&Add_Float(
     A=&RandomRangeFloat(Minimum=0, Maximum=1)
     B=&FloatFromCurve(FloatCurve=DataInterface() {propertyValues {Curve {keys [
          {arriveTangent=2.7533, interpMode=RCIM_Cubic, leaveTangent=1.8958, tangentMode=RCTM_User, time=0, value=1}
          {arriveTangent=0.8292, interpMode=RCIM_Cubic, leaveTangent=0.8292, tangentMode=RCTM_User, time=0.321, value=1.3453}
          {arriveTangent=0.0529, interpMode=RCIM_Cubic, leaveTangent=0.2021, tangentMode=RCTM_User, time=1, value=1.5}
         ]}}})
    )
    "Sphere Normal (NormalMultiplier)":NiagaraFloat=&FloatFromCurve(FloatCurve=DataInterface() {propertyValues {Curve {keys [<at 0.5 1>,<at 1 0>]}}})
    "Uniform Scale Noise (DissolveNoiseScale)":NiagaraFloat=&RandomRangeFloat(Minimum=1, Maximum=3)
    "Uniform Scale Noise (NoiseScale)":NiagaraFloat=&RandomRangeFloat(Minimum=1.1, Maximum=1.4)
   ) {module=Smoke_V2_Material_Dynamics}
   &CameraOffset(Camera Offset Amount=15)
  ]
  rendererClasses [&NiagaraSpriteRendererProperties]
  renderers [
   {rendererClass=&NiagaraSpriteRendererProperties}
   {rendererClass=&NiagaraSpriteRendererProperties, rendererIndex=1}
  ]
 }
 SmokeGround {
  emitterUpdateScript [
   &EmitterState()
   &SpawnBurst_Instantaneous(Spawn Count=1, Use Spawn Probability=false)
  ]
  particleSpawnScript [
   &InitializeParticle(
    Color=<rgba 0.374 0.9791 1 1>
    Lifetime=0.25
    Module.Write Color=true
    Module.Write Lifetime=true
    Module.Write Mass=true
    Module.Write Material Random=true
    Module.Write Position=false
    Module.Write Scale=false
    Module.Write SpriteRotation=true
    Module.Write SpriteSize=true
    Position=&SimulationPosition()
    Sprite Rotation=&RandomRangeFloat(Minimum=0, Maximum=1)
    Sprite Size=&Vector2DFromFloat(Value=&Lerp_Float(A=100, B=200, Alpha=$System.NormalizedStageNum))
   )
   /SetVariables(
    Particles.SpriteFacing=<0 0 1>
    Particles.SpriteRotationAlt:NiagaraFloat=&RandomRangeFloat(Minimum=0, Maximum=1)
   ) {isSetParameters=true}
   &SystemLocation(MaskSpawnGroup=false, Offset=<0 0 5>, Offset Coordinate Space=::World)
  ]
  particleUpdateScript [
   &ParticleState()
   &ScaleSpriteSize(Scale Factor=&Vector2DFromFloat(Value=&FloatFromCurve(FloatCurve=DataInterface() {propertyValues {Curve {keys [
          {arriveTangent=5.4041, interpMode=RCIM_Cubic, leaveTangent=5.4041, tangentMode=RCTM_User, time=-0.0046, value=0.4}
          {arriveTangent=0.6903, interpMode=RCIM_Cubic, leaveTangent=0.6903, tangentMode=RCTM_User, time=1.0011, value=3.4809}
         ]}}})))
   &DynamicMaterialParameters(
    Index 0 Param 1=1
    Index 0 Param 2=5
    Index 0 Param 3=1
    Index 0 Param 4=1
    Index 1 Param 1=&FloatFromCurve(FloatCurve=DataInterface() {propertyValues {Curve {keys [{arriveTangent=22.1134, interpMode=RCIM_Cubic, time=0.3, value=4}, {interpMode=RCIM_Cubic, leaveTangent=6.3587, time=1, value=12}]}}})
    Index 1 Param 2=&Lerp_Float(A=0.01, B=0.22, Alpha=$Particles.NormalizedAge)
    Index 1 Param 3=2
    Module.Param0WriteEnabled=true
    Module.Param0WriteEnabled1=true
    Module.Param1WriteEnabled=true
    Module.Param1WriteEnabled1=true
    Module.Param2WriteEnabled=true
    Module.Param2WriteEnabled1=true
    Module.Param3WriteEnabled=true
    Module.Param3WriteEnabled1=true
    Write Parameter Index 1=true
   )
   &ScaleColor(
    Scale Alpha=&FloatFromCurve(
     FloatCurve=DataInterface() {propertyValues {Curve {keys [
         <at 0 1 RCIM_Cubic>
         {arriveTangent=-0.3825, interpMode=RCIM_Cubic, leaveTangent=-0.3825, tangentMode=RCTM_User, time=0.1284, value=0.1149}
         {arriveTangent=-0.2386, interpMode=RCIM_Cubic, leaveTangent=-0.2386, tangentMode=RCTM_User, time=1.0011, value=0.002}
        ]}}}
     Scale Curve=&Lerp_Float(A=1, B=4, Alpha=$System.NormalizedStageNum)
    )
    ScaleA=true
    ScaleRGB=false
   )
   &CameraOffset(Camera Offset Amount=50) {enabled=false}
  ]
  rendererClasses [&NiagaraSpriteRendererProperties]
  renderers [
   {rendererClass=&NiagaraSpriteRendererProperties}
   {rendererClass=&NiagaraSpriteRendererProperties, rendererIndex=1}
  ]
 }
 SmokeLinger {
  emitterUpdateScript [
   &EmitterState()
   &SpawnBurst_Instantaneous(Spawn Count=10, Use Spawn Probability=false)
  ]
  particleSpawnScript [
   /SetVariables(Particles.InitSpriteRotRate=&UniformRangedFloat(Minimum=-3, Maximum=3)) {isSetParameters=true, module=SetVariables_#2}
   &InitializeParticle(
    Color=<rgba 0.5 0.5 0.5 1>
    Lifetime=&RandomRangeFloat(Minimum=0.75, Maximum=2)
    Module.Write Color=true
    Module.Write Lifetime=true
    Module.Write Mass=true
    Module.Write Material Random=true
    Module.Write Position=true
    Module.Write Scale=false
    Module.Write SpriteRotation=true
    Module.Write SpriteSize=true
    Sprite Rotation=&RandomRangeFloat(Minimum=-360, Maximum=360)
    Sprite Size=&Vector2DFromFloat(Value=&Multiply_Float(
      A=&Lerp_Float(A=150, B=300, Alpha=&UniformRangedFloat(Minimum=0, Maximum=1))
      B=&Lerp_Float(A=0.6, B=1, Alpha=$System.NormalizedStageNum)
     ))
   )
   &SphereLocation(
    MaskSpawnGroup=false
    Non Uniform Scale=<1 1 0.5>
    Override Seed=true
    Random Seed=20
    Sphere Coordinate Space=::World
    Sphere Radius=&Lerp_Float(A=10, B=30, Alpha=&Lerp_Float(A=0.5, B=1, Alpha=$System.NormalizedStageNum))
    Surface Only=false
    UseNonUniformScale=true
    UseOffset=false
    UseSphereOrientation=false
   )
   &AddVelocityFromPoint(
    Module.Use Falloff Distance=false
    Velocity Strength=&Multiply_Float(
     A=&RandomRangeFloat(Minimum=200, Maximum=500)
     B=&Lerp_Float(A=0.7, B=1, Alpha=$System.NormalizedStageNum)
    )
   )
   /SetVariables(Particles.Velocity=&Multiply_Vector(
     A=&MakeVector(
      X=&MakeFloatFromVector(Vector=$Particles.Velocity, Channel=::NewEnumerator0)
      Y=&MakeFloatFromVector(Vector=$Particles.Velocity, Channel=::NewEnumerator1)
      Z=&Abs_Float(Float:NiagaraFloat=&MakeFloatFromVector(Vector=$Particles.Velocity, Channel=::NewEnumerator2))
     )
     B=<1 1 0.5>
    )) {isSetParameters=true, module=SetVariables_#3}
  ]
  particleUpdateScript [
   &ParticleState()
   &SpriteRotationRate(Rotation Rate=&Multiply_Float(
     A=$Particles.InitSpriteRotRate
     B=&FloatFromCurve(
      FloatCurve=DataInterface() {propertyValues {Curve {keys [
          {arriveTangent=-4.0107, interpMode=RCIM_Cubic, leaveTangent=-4.0107, tangentMode=RCTM_User, time=-0.0046, value=1.0113}
          {arriveTangent=-0.2496, interpMode=RCIM_Cubic, leaveTangent=-0.2496, tangentMode=RCTM_User, time=0.5, value=0.16}
         ]}}}
      CurveIndex=$Emitter.Age
      Scale Curve=150
     )
    ))
   &ScaleColor(
    Scale Alpha=&FloatFromCurve(
     FloatCurve=DataInterface() {propertyValues {Curve {keys [
         {arriveTangent=-7.3919, interpMode=RCIM_Cubic, leaveTangent=-7.3919, tangentMode=RCTM_User, time=0.1213, value=0.984}
         {arriveTangent=-0.6256, interpMode=RCIM_Cubic, leaveTangent=-0.6256, tangentMode=RCTM_User, time=0.3214, value=0.2286}
         <at 1 0>
        ]}}}
     Scale Curve=&Lerp_Float(A=0.25, B=1, Alpha=$System.NormalizedStageNum)
    )
    Scale RGB=&MakeVectorFromLinearColorRGB(Color=&ColorFromCurve(ColorCurve=DataInterface() {propertyValues {
        AlphaCurve {keys [<at 0 1 RCIM_Cubic>]}
        BlueCurve {keys [<at 0 0.3385 RCIM_Cubic>, {arriveTangent=-0.131, interpMode=RCIM_Cubic, leaveTangent=-0.131, time=0.06, value=0.5}, <at 1.0334 0.2031>]}
        GreenCurve {keys [<at 0 0.2204 RCIM_Cubic>, {arriveTangent=-0.0167, interpMode=RCIM_Cubic, leaveTangent=-0.0167, time=0.06, value=0.375}, <at 1.0334 0.2031>]}
        RedCurve {keys [<at 0 0.269 RCIM_Cubic>, {arriveTangent=-0.0637, interpMode=RCIM_Cubic, leaveTangent=-0.0637, time=0.06, value=0.3958}, <at 1.0334 0.2031>]}
       }}))
    ScaleA=true
    ScaleRGB=true
   )
   &Drag(Drag=4, Use Linear Drag=true, Use Rotational Drag=false)
   &SolveForcesAndVelocity(Clamp Velocity=false, Limit Acceleration=false)
   &ScaleSpriteSize(Scale Factor=&Vector2DFromFloat(Value=&FloatFromCurve(FloatCurve=DataInterface() {propertyValues {Curve {keys [
          {arriveTangent=2.8475, interpMode=RCIM_Cubic, leaveTangent=2.3251, tangentMode=RCTM_User, time=0, value=0.15}
          {arriveTangent=1.1483, interpMode=RCIM_Cubic, leaveTangent=1.1483, tangentMode=RCTM_User, time=0.2651, value=0.6514}
          {arriveTangent=0.2754, interpMode=RCIM_Cubic, leaveTangent=0.2754, tangentMode=RCTM_User, time=1.0463, value=1}
         ]}}})))
   &self:Smoke_V2_Material_Dynamics(
    "Apply Noise (DissolveLerpNoise)"=1
    "Apply Noise (LerpNoise)"=1
    "Erode Alpha (DissolveNoiseErosionPower)":NiagaraFloat=&RandomRangeFloat(Minimum=0.5, Maximum=1)
    "Erode Emissive (NoiseErosionPower)":NiagaraFloat=&FloatFromCurve(FloatCurve=DataInterface() {propertyValues {Curve {keys [
         {arriveTangent=33.5848, interpMode=RCIM_Cubic, leaveTangent=33.5848, tangentMode=RCTM_User, time=0, value=0}
         {arriveTangent=32.1053, interpMode=RCIM_Cubic, leaveTangent=32.1053, tangentMode=RCTM_User, time=0.15, value=2}
        ]}}})
    "Intensity Modifier (Emissive)":NiagaraFloat=&FloatFromCurve(
     FloatCurve=DataInterface() {propertyValues {Curve {keys [{interpMode=RCIM_Cubic, tangentMode=RCTM_User, time=0, value=1}, <at 0.15 0>]}}}
     Scale Curve=&SelectFloatFromArray(Direct Array Index=$User.StageNum, Float Selection Array=DataInterface() {propertyValues {FloatData [0.01,0.05,0.5]}})
    )
    "Noise Offset X (DissolveNoiseOffsetX)":NiagaraFloat=&RandomRangeFloat(Minimum=0, Maximum=1)
    "Noise Offset X (NoiseOffsetX)":NiagaraFloat=&Add_Float(
     A=&RandomRangeFloat(Minimum=0, Maximum=1)
     B=&FloatFromCurve(
      FloatCurve=DataInterface() {propertyValues {Curve {keys [
          {arriveTangent=12.1425, interpMode=RCIM_Cubic, leaveTangent=6.464, tangentMode=RCTM_User, time=0, value=0}
          {arriveTangent=0.0529, interpMode=RCIM_Cubic, leaveTangent=0.2021, tangentMode=RCTM_User, time=0.15, value=0.5}
         ]}}}
      Scale Curve=&RandomRangeFloat(Minimum=0.25, Maximum=1)
     )
    )
    "Normalized Frame Number (Animation)":NiagaraFloat=&Add_Float(
     A=&RandomRangeFloat(Minimum=0, Maximum=1)
     B=&FloatFromCurve(FloatCurve=DataInterface() {propertyValues {Curve {keys [
          {arriveTangent=2.5454, interpMode=RCIM_Cubic, leaveTangent=2.266, tangentMode=RCTM_User, time=0, value=1}
          {arriveTangent=0.0529, interpMode=RCIM_Cubic, leaveTangent=0.2021, tangentMode=RCTM_User, time=1, value=2}
         ]}}})
    )
    "Sphere Normal (NormalMultiplier)":NiagaraFloat=&FloatFromCurve(FloatCurve=DataInterface() {propertyValues {Curve {keys [<at 0.5 1>,<at 1 0>]}}})
    "Uniform Scale Noise (DissolveNoiseScale)":NiagaraFloat=&RandomRangeFloat(Minimum=1, Maximum=3)
    "Uniform Scale Noise (NoiseScale)":NiagaraFloat=&RandomRangeFloat(Minimum=1.1, Maximum=1.4)
   ) {module=Smoke_V2_Material_Dynamics}
   &CameraOffset(Camera Offset Amount=15)
  ]
  rendererClasses [&NiagaraSpriteRendererProperties]
  renderers [
   {rendererClass=&NiagaraSpriteRendererProperties}
   {rendererClass=&NiagaraSpriteRendererProperties, rendererIndex=1}
  ]
 }
 Smoke_Sphere001 {
  bEnabled=false
  emitterUpdateScript [
   &EmitterState()
   &SpawnBurst_Instantaneous(Spawn Count=30, Use Spawn Probability=false)
  ]
  particleSpawnScript [
   /SetVariables(
    Particles.InitSpriteRotRate=&UniformRangedFloat(Minimum=-3, Maximum=3)
    Transient.ScalePicker:NiagaraFloat=&UniformRangedFloat(Minimum=0, Maximum=1)
   ) {isSetParameters=true, module=SetVariables_#2}
   &InitializeParticle(
    Color=<rgba 0.5 0.5 0.5 1>
    Lifetime=&Lerp_Float(
     A=&Lerp_Float(A=0.4, B=0.75, Alpha=$Transient.ScalePicker)
     B=&Lerp_Float(A=0.2, B=0.5, Alpha=$Transient.ScalePicker)
     Alpha=$System.NormalizedStageNum
    )
    Module.Write Color=true
    Module.Write Lifetime=true
    Module.Write Mass=true
    Module.Write Material Random=true
    Module.Write Position=true
    Module.Write Scale=false
    Module.Write SpriteRotation=true
    Module.Write SpriteSize=true
    Sprite Rotation=&RandomRangeFloat(Minimum=-360, Maximum=360)
    Sprite Size=&Vector2DFromFloat(Value=&Multiply_Float(
      A=&Lerp_Float(
       A=75
       B=175
       Alpha=&ClampFloat(Float=&Add_Float(A=$Transient.ScalePicker, B=&RandomRangeFloat(Minimum=-0.2, Maximum=0.3)))
      )
      B=&Lerp_Float(A=0.6, B=1, Alpha=$System.NormalizedStageNum)
     ))
   )
   &SphereLocation(
    MaskSpawnGroup=false
    Non Uniform Scale=<1 1 0.75>
    Override Seed=true
    Random Seed=20
    Sphere Coordinate Space=::World
    Sphere Radius=&Lerp_Float(A=10, B=50, Alpha=&Lerp_Float(A=0.5, B=1, Alpha=$System.NormalizedStageNum))
    Surface Only=false
    UseNonUniformScale=true
    UseOffset=false
    UseSphereOrientation=false
   )
   &AddVelocityFromPoint(
    Module.Use Falloff Distance=false
    Velocity Strength=&Multiply_Float(
     A=&RandomRangeFloat(Minimum=50, Maximum=500)
     B=&Lerp_Float(A=0.7, B=1, Alpha=$System.NormalizedStageNum)
    )
   )
   /SetVariables(Particles.Velocity=&Multiply_Vector(
     A=&MakeVector(
      X=&MakeFloatFromVector(Vector=$Particles.Velocity, Channel=::NewEnumerator0)
      Y=&MakeFloatFromVector(Vector=$Particles.Velocity, Channel=::NewEnumerator1)
      Z=&Abs_Float(Float:NiagaraFloat=&MakeFloatFromVector(Vector=$Particles.Velocity, Channel=::NewEnumerator2))
     )
     B=<2 2 1>
    )) {isSetParameters=true, module=SetVariables_#3}
  ]
  particleUpdateScript [
   &ParticleState()
   &SpriteRotationRate(Rotation Rate=&Multiply_Float(
     A=$Particles.InitSpriteRotRate
     B=&FloatFromCurve(
      FloatCurve=DataInterface() {propertyValues {Curve {keys [
          {arriveTangent=-4.0107, interpMode=RCIM_Cubic, leaveTangent=-4.0107, tangentMode=RCTM_User, time=-0.0046, value=1.0113}
          {arriveTangent=-0.2496, interpMode=RCIM_Cubic, leaveTangent=-0.2496, tangentMode=RCTM_User, time=0.5, value=0.16}
         ]}}}
      CurveIndex=$Emitter.Age
      Scale Curve=150
     )
    ))
   &ScaleColor(
    Scale Alpha=&FloatFromCurve(FloatCurve=DataInterface() {propertyValues {Curve {keys [
         {arriveTangent=-22.8339, interpMode=RCIM_Cubic, leaveTangent=-22.8339, tangentMode=RCTM_User, time=0.1219, value=1}
         {arriveTangent=-3.3275, interpMode=RCIM_Cubic, leaveTangent=-3.3275, tangentMode=RCTM_User, time=0.1573, value=0.2774}
         {arriveTangent=-0.489, interpMode=RCIM_Cubic, leaveTangent=-0.489, tangentMode=RCTM_User, time=0.2679, value=0.1132}
         <at 1 0>
        ]}}})
    Scale RGB=&MakeVectorFromLinearColorRGB(Color=&ColorFromCurve(ColorCurve=DataInterface() {propertyValues {
        AlphaCurve {keys [<at 0 1 RCIM_Cubic>]}
        BlueCurve {keys [<at 0 0.3385 RCIM_Cubic>, {arriveTangent=-0.131, interpMode=RCIM_Cubic, leaveTangent=-0.131, time=0.2245, value=0.5}, <at 1.0334 0.2031>]}
        GreenCurve {keys [<at 0 0.2204 RCIM_Cubic>, {arriveTangent=-0.0167, interpMode=RCIM_Cubic, leaveTangent=-0.0167, time=0.2245, value=0.375}, <at 1.0334 0.2031>]}
        RedCurve {keys [<at 0 0.269 RCIM_Cubic>, {arriveTangent=-0.0637, interpMode=RCIM_Cubic, leaveTangent=-0.0637, time=0.2245, value=0.3958}, <at 1.0334 0.2031>]}
       }}))
    ScaleA=true
    ScaleRGB=true
   )
   &CurlNoiseForce(Noise Frequency=0.1, Noise Strength=800, Pan Noise=true, Pan Noise Field=<0 0 1>)
   &Drag(Drag=3.5, Use Linear Drag=true, Use Rotational Drag=false)
   &SolveForcesAndVelocity(Clamp Velocity=false, Limit Acceleration=false)
   &ScaleSpriteSize(Scale Factor=&Vector2DFromFloat(Value=&FloatFromCurve(FloatCurve=DataInterface() {propertyValues {Curve {keys [
          {arriveTangent=2.8475, interpMode=RCIM_Cubic, leaveTangent=2.3251, tangentMode=RCTM_User, time=0, value=0.15}
          {arriveTangent=1.1483, interpMode=RCIM_Cubic, leaveTangent=1.1483, tangentMode=RCTM_User, time=0.2651, value=0.6514}
          {arriveTangent=0.2754, interpMode=RCIM_Cubic, leaveTangent=0.2754, tangentMode=RCTM_User, time=1.0463, value=1}
         ]}}})))
   &self:Smoke_V2_Material_Dynamics(
    "Apply Noise (DissolveLerpNoise)"=1
    "Apply Noise (LerpNoise)"=1
    "Erode Alpha (DissolveNoiseErosionPower)":NiagaraFloat=&RandomRangeFloat(Minimum=0.75, Maximum=1)
    "Erode Emissive (NoiseErosionPower)":NiagaraFloat=&FloatFromCurve(FloatCurve=DataInterface() {propertyValues {Curve {keys [
         {arriveTangent=2.8986, interpMode=RCIM_Cubic, leaveTangent=2.8986, tangentMode=RCTM_User, time=0, value=0}
         {arriveTangent=5.2006, interpMode=RCIM_Cubic, leaveTangent=5.2006, tangentMode=RCTM_User, time=1, value=2}
        ]}}})
    "Intensity Modifier (Emissive)":NiagaraFloat=&FloatFromCurve(
     FloatCurve=DataInterface() {propertyValues {Curve {keys [{interpMode=RCIM_Cubic, tangentMode=RCTM_User, time=0, value=1}, <at 1 0>]}}}
     Scale Curve=&MakeCustomFloatFromBool(
      Bool=&SetBoolByIntComparison(A=$User.StageNum, B=0, Comparison Type=::NewEnumerator4)
      True Float=0.01
      False Float=&MakeCustomFloatFromBool(
       Bool=&SetBoolByIntComparison(A=$User.StageNum, B=1, Comparison Type=::NewEnumerator4)
       True Float=0.05
       False Float=0.5
      )
     )
    )
    "Noise Offset X (DissolveNoiseOffsetX)":NiagaraFloat=&RandomRangeFloat(Minimum=0, Maximum=1)
    "Noise Offset X (NoiseOffsetX)":NiagaraFloat=&Add_Float(
     A=&RandomRangeFloat(Minimum=0, Maximum=1)
     B=&FloatFromCurve(
      FloatCurve=DataInterface() {propertyValues {Curve {keys [
          {arriveTangent=1.5771, interpMode=RCIM_Cubic, leaveTangent=1.223, tangentMode=RCTM_User, time=0, value=0}
          {arriveTangent=0.5984, interpMode=RCIM_Cubic, leaveTangent=0.5984, tangentMode=RCTM_User, time=0.321, value=0.3333}
          {arriveTangent=0.0529, interpMode=RCIM_Cubic, leaveTangent=0.2021, tangentMode=RCTM_User, time=1, value=0.5}
         ]}}}
      Scale Curve=&RandomRangeFloat(Minimum=0.25, Maximum=1)
     )
    )
    "Normalized Frame Number (Animation)":NiagaraFloat=&Add_Float(
     A=&RandomRangeFloat(Minimum=0, Maximum=1)
     B=&FloatFromCurve(FloatCurve=DataInterface() {propertyValues {Curve {keys [
          {arriveTangent=2.7533, interpMode=RCIM_Cubic, leaveTangent=1.8958, tangentMode=RCTM_User, time=0, value=1}
          {arriveTangent=0.8292, interpMode=RCIM_Cubic, leaveTangent=0.8292, tangentMode=RCTM_User, time=0.321, value=1.3453}
          {arriveTangent=0.0529, interpMode=RCIM_Cubic, leaveTangent=0.2021, tangentMode=RCTM_User, time=1, value=1.5}
         ]}}})
    )
    "Sphere Normal (NormalMultiplier)":NiagaraFloat=&FloatFromCurve(FloatCurve=DataInterface() {propertyValues {Curve {keys [<at 0.5 1>,<at 1 0>]}}})
    "Uniform Scale Noise (DissolveNoiseScale)":NiagaraFloat=&RandomRangeFloat(Minimum=1.5, Maximum=2)
    "Uniform Scale Noise (NoiseScale)":NiagaraFloat=&RandomRangeFloat(Minimum=1.1, Maximum=1.4)
   ) {module=Smoke_V2_Material_Dynamics}
   &self:Smoke_V2_Material_Dynamics(
    "Apply Noise (DissolveLerpNoise)"=1
    "Apply Noise (LerpNoise)"=1
    "Erode Alpha (DissolveNoiseErosionPower)":NiagaraFloat=&RandomRangeFloat(Minimum=0.75, Maximum=1)
    "Erode Emissive (NoiseErosionPower)":NiagaraFloat=&FloatFromCurve(FloatCurve=DataInterface() {propertyValues {Curve {keys [
         {arriveTangent=2.8986, interpMode=RCIM_Cubic, leaveTangent=2.8986, tangentMode=RCTM_User, time=0, value=0}
         {arriveTangent=5.2006, interpMode=RCIM_Cubic, leaveTangent=5.2006, tangentMode=RCTM_User, time=1, value=2}
        ]}}})
    "Intensity Modifier (Emissive)":NiagaraFloat=&FloatFromCurve(
     FloatCurve=DataInterface() {propertyValues {Curve {keys [{interpMode=RCIM_Cubic, tangentMode=RCTM_User, time=0, value=1}, <at 1 0>]}}}
     Scale Curve=&MakeCustomFloatFromBool(
      Bool=&SetBoolByIntComparison(A=$User.StageNum, B=0, Comparison Type=::NewEnumerator4)
      True Float=0.01
      False Float=&MakeCustomFloatFromBool(
       Bool=&SetBoolByIntComparison(A=$User.StageNum, B=1, Comparison Type=::NewEnumerator4)
       True Float=0.05
       False Float=0.5
      )
     )
    )
    "Noise Offset X (DissolveNoiseOffsetX)":NiagaraFloat=&RandomRangeFloat(Minimum=0, Maximum=1)
    "Noise Offset X (NoiseOffsetX)":NiagaraFloat=&FloatFromCurve(
     FloatCurve=DataInterface() {propertyValues {Curve {keys [<at 0 0.2>, {leaveTangent=0.9066, time=1, value=2}]}}}
     Scale Curve=&RandomRangeFloat(Minimum=0.5, Maximum=1)
    )
    "Normalized Frame Number (Animation)":NiagaraFloat=&Add_Float(
     A=&RandomRangeFloat(Minimum=0, Maximum=1)
     B=&FloatFromCurve(FloatCurve=DataInterface() {propertyValues {Curve {keys [
          {arriveTangent=1.6548, interpMode=RCIM_Cubic, leaveTangent=1.268, tangentMode=RCTM_User, time=0, value=0}
          {arriveTangent=0.6707, interpMode=RCIM_Cubic, leaveTangent=0.6707, tangentMode=RCTM_User, time=0.321, value=0.3453}
          {arriveTangent=0.0529, interpMode=RCIM_Cubic, leaveTangent=0.2021, tangentMode=RCTM_User, time=1, value=0.5}
         ]}}})
    )
    "Sphere Normal (NormalMultiplier)":NiagaraFloat=&FloatFromCurve(FloatCurve=DataInterface() {propertyValues {Curve {keys [<at 0.5 1>,<at 1 0>]}}})
    "Uniform Scale Noise (DissolveNoiseScale)":NiagaraFloat=&RandomRangeFloat(Minimum=1.5, Maximum=2)
    "Uniform Scale Noise (NoiseScale)":NiagaraFloat=&FloatFromCurve(
     FloatCurve=DataInterface() {propertyValues {Curve {keys [{arriveTangent=-0.136, interpMode=RCIM_Cubic, leaveTangent=-0.136, tangentMode=RCTM_User, time=0, value=1}, <at 1 0.95>]}}}
     Scale Curve=&RandomRangeFloat(Minimum=1.1, Maximum=1.4)
    )
   ) {enabled=false, module=Smoke_V2_Material_Dynamics002}
   &self:Smoke_V2_Material_Dynamics(
    "Apply Noise (DissolveLerpNoise)":NiagaraFloat=&FloatFromCurve(FloatCurve=DataInterface() {propertyValues {Curve {keys [
         {interpMode=RCIM_Cubic, leaveTangent=4.2485, tangentMode=RCTM_Break, time=0, value=0}
         {arriveTangent=1.0548, interpMode=RCIM_Cubic, leaveTangent=1.0548, tangentMode=RCTM_User, time=0.17, value=0.5715}
         <at 1 1>
        ]}}})
    "Apply Noise (LerpNoise)":NiagaraFloat=&FloatFromCurve(FloatCurve=DataInterface() {propertyValues {Curve {keys [<at 0 0>, {arriveTangent=12.7455, interpMode=RCIM_Cubic, leaveTangent=12.7455, tangentMode=RCTM_User, time=0.25, value=1.5}]}}})
    "Erode Alpha (DissolveNoiseErosionPower)":NiagaraFloat=&FloatFromCurve(FloatCurve=DataInterface() {propertyValues {Curve {keys [
         {interpMode=RCIM_Cubic, leaveTangent=2.3064, tangentMode=RCTM_Break, time=0, value=1}
         {interpMode=RCIM_Cubic, leaveTangent=11.1233, tangentMode=RCTM_Break, time=1, value=2}
        ]}}})
    "Erode Emissive (NoiseErosionPower)":NiagaraFloat=&FloatFromCurve(FloatCurve=DataInterface() {propertyValues {Curve {keys [
         {interpMode=RCIM_Cubic, tangentMode=RCTM_User, time=0, value=0}
         {arriveTangent=55.3407, interpMode=RCIM_Cubic, leaveTangent=55.3407, tangentMode=RCTM_User, time=0.2, value=5}
        ]}}})
    "Intensity Modifier (Emissive)":NiagaraFloat=&FloatFromCurve(
     FloatCurve=DataInterface() {propertyValues {Curve {keys [{interpMode=RCIM_Cubic, tangentMode=RCTM_User, time=0, value=1}, <at 0.2 0>]}}}
     Scale Curve=&MakeCustomFloatFromBool(
      Bool=&SetBoolByIntComparison(A=$User.StageNum, B=0, Comparison Type=::NewEnumerator4)
      True Float=0.01
      False Float=&MakeCustomFloatFromBool(
       Bool=&SetBoolByIntComparison(A=$User.StageNum, B=1, Comparison Type=::NewEnumerator4)
       True Float=0.05
       False Float=1
      )
     )
    )
    "Noise Offset X (DissolveNoiseOffsetX)":NiagaraFloat=&FloatFromCurve(
     FloatCurve=DataInterface() {propertyValues {Curve {keys [
         {arriveTangent=1.9056, interpMode=RCIM_Cubic, leaveTangent=1.9056, tangentMode=RCTM_User, time=0, value=0.2}
         {arriveTangent=0.0259, interpMode=RCIM_Cubic, leaveTangent=0.0259, tangentMode=RCTM_User, time=1, value=1.4}
        ]}}}
     Scale Curve=&RandomRangeFloat(Minimum=0.2, Maximum=0.7)
    )
    "Noise Offset X (NoiseOffsetX)":NiagaraFloat=&FloatFromCurve(
     FloatCurve=DataInterface() {propertyValues {Curve {keys [<at 0 0.2>, {leaveTangent=0.9066, time=1, value=2}]}}}
     Scale Curve=&RandomRangeFloat(Minimum=0.5, Maximum=1)
    )
    "Normalized Frame Number (Animation)":NiagaraFloat=&Add_Float(
     A=&RandomRangeFloat(Minimum=0, Maximum=1)
     B=&FloatFromCurve(FloatCurve=DataInterface() {propertyValues {Curve {keys [
          {arriveTangent=2.7533, interpMode=RCIM_Cubic, leaveTangent=1.8958, tangentMode=RCTM_User, time=0, value=1}
          {arriveTangent=0.8292, interpMode=RCIM_Cubic, leaveTangent=0.8292, tangentMode=RCTM_User, time=0.321, value=1.3453}
          {arriveTangent=0.0529, interpMode=RCIM_Cubic, leaveTangent=0.2021, tangentMode=RCTM_User, time=1, value=1.5}
         ]}}})
    )
    "Sphere Normal (NormalMultiplier)":NiagaraFloat=&FloatFromCurve(FloatCurve=DataInterface() {propertyValues {Curve {keys [<at 0.5 1>,<at 1 0>]}}})
    "Uniform Scale Noise (DissolveNoiseScale)":NiagaraFloat=&FloatFromCurve(
     FloatCurve=DataInterface() {propertyValues {Curve {keys [
         {arriveTangent=-0.6867, interpMode=RCIM_Cubic, leaveTangent=-0.6867, tangentMode=RCTM_User, time=0, value=1}
         {interpMode=RCIM_Cubic, leaveTangent=-1.2956, tangentMode=RCTM_Break, time=1, value=0.75}
        ]}}}
     Scale Curve=1.6
    )
    "Uniform Scale Noise (NoiseScale)":NiagaraFloat=&FloatFromCurve(
     FloatCurve=DataInterface() {propertyValues {Curve {keys [{arriveTangent=-0.136, interpMode=RCIM_Cubic, leaveTangent=-0.136, tangentMode=RCTM_User, time=0, value=1}, <at 1 0.95>]}}}
     Scale Curve=&RandomRangeFloat(Minimum=1.1, Maximum=1.4)
    )
   ) {enabled=false, module=Smoke_V2_Material_Dynamics001}
   &DynamicMaterialParameters(
    Index 0 Param 1=&Add_Float(
     A=&RandomRangeFloat(Minimum=0, Maximum=1)
     B=&FloatFromCurve(
      FloatCurve=DataInterface() {propertyValues {Curve {keys [
          {arriveTangent=2.7533, interpMode=RCIM_Cubic, leaveTangent=1.8958, tangentMode=RCTM_User, time=0, value=1}
          {arriveTangent=0.8292, interpMode=RCIM_Cubic, leaveTangent=0.8292, tangentMode=RCTM_User, time=0.321, value=1.3453}
          {arriveTangent=0.0529, interpMode=RCIM_Cubic, leaveTangent=0.2021, tangentMode=RCTM_User, time=1, value=1.5}
         ]}}}
      CurveIndex=&Add_Float(A=&Multiply_Float(A=$Particles.Initial.MaterialRandom, B=0.5), B=$Particles.NormalizedAge)
     )
    )
    Index 0 Param 3=&FloatFromCurve(
     FloatCurve=DataInterface() {propertyValues {Curve {keys [
         {arriveTangent=-6567.626, interpMode=RCIM_Cubic, time=0, value=1}
         {arriveTangent=-15.6335, interpMode=RCIM_Cubic, leaveTangent=0.2723, tangentMode=RCTM_Break, time=0.1, value=0}
        ]}}}
     CurveIndex=&Multiply_Float(A=$Particles.NormalizedAge, B=1)
    )
    Index 0 Param 4=&FloatFromCurve(FloatCurve=DataInterface() {propertyValues {Curve {keys [<at 0.5 1>,<at 1 0>]}}})
    Index 1 Param 1=&FloatFromCurve(
     FloatCurve=DataInterface() {propertyValues {Curve {keys [<at 0 0>, {arriveTangent=12.7455, interpMode=RCIM_Cubic, leaveTangent=12.7455, tangentMode=RCTM_User, time=0.25, value=1}]}}}
     Scale Curve=&RandomRangeFloat(Minimum=1.5, Maximum=1.5)
    )
    Index 1 Param 2=&FloatFromCurve(
     FloatCurve=DataInterface() {propertyValues {Curve {keys [<at 0 0.2>, {leaveTangent=0.9066, time=1, value=2}]}}}
     Scale Curve=&RandomRangeFloat(Minimum=0.5, Maximum=0.6)
    )
    Index 1 Param 3=&FloatFromCurve(FloatCurve=DataInterface() {propertyValues {Curve {keys [
         {interpMode=RCIM_Cubic, leaveTangent=6.3917, tangentMode=RCTM_Break, time=0, value=0.2}
         {arriveTangent=1.7549, interpMode=RCIM_Cubic, leaveTangent=10.6767, tangentMode=RCTM_User, time=0.25, value=1}
        ]}}})
    Index 1 Param 4=&FloatFromCurve(
     FloatCurve=DataInterface() {propertyValues {Curve {keys [{arriveTangent=-0.136, interpMode=RCIM_Cubic, leaveTangent=-0.136, tangentMode=RCTM_User, time=0, value=1}, <at 1 0.95>]}}}
     Scale Curve=&RandomRangeFloat(Minimum=1.1, Maximum=1.4)
    )
    Index 2 Param 1=&FloatFromCurve(FloatCurve=DataInterface() {propertyValues {Curve {keys [
         {interpMode=RCIM_Cubic, leaveTangent=4.2485, tangentMode=RCTM_Break, time=0, value=0}
         {arriveTangent=1.0548, interpMode=RCIM_Cubic, leaveTangent=1.0548, tangentMode=RCTM_User, time=0.17, value=0.5715}
         <at 1 1>
        ]}}})
    Index 2 Param 2=&FloatFromCurve(
     FloatCurve=DataInterface() {propertyValues {Curve {keys [
         {arriveTangent=1.9056, interpMode=RCIM_Cubic, leaveTangent=1.9056, tangentMode=RCTM_User, time=0, value=0.2}
         {arriveTangent=0.0259, interpMode=RCIM_Cubic, leaveTangent=0.0259, tangentMode=RCTM_User, time=1, value=1.4}
        ]}}}
     Scale Curve=&RandomRangeFloat(Minimum=0.2, Maximum=0.7)
    )
    Index 2 Param 3=&FloatFromCurve(FloatCurve=DataInterface() {propertyValues {Curve {keys [
         {interpMode=RCIM_Cubic, leaveTangent=2.3064, tangentMode=RCTM_Break, time=0, value=1}
         {interpMode=RCIM_Cubic, leaveTangent=11.1233, tangentMode=RCTM_Break, time=1, value=2}
        ]}}})
    Index 2 Param 4=&FloatFromCurve(
     FloatCurve=DataInterface() {propertyValues {Curve {keys [
         {arriveTangent=-0.6867, interpMode=RCIM_Cubic, leaveTangent=-0.6867, tangentMode=RCTM_User, time=0, value=1}
         {interpMode=RCIM_Cubic, leaveTangent=-1.2956, tangentMode=RCTM_Break, time=1, value=0.75}
        ]}}}
     Scale Curve=1.6
    )
    Module.Param0WriteEnabled=true
    Module.Param0WriteEnabled1=true
    Module.Param0WriteEnabled2=true
    Module.Param1WriteEnabled=false
    Module.Param1WriteEnabled1=true
    Module.Param1WriteEnabled2=true
    Module.Param2WriteEnabled=true
    Module.Param2WriteEnabled1=true
    Module.Param2WriteEnabled2=true
    Module.Param3WriteEnabled=true
    Module.Param3WriteEnabled1=true
    Module.Param3WriteEnabled2=true
    Write Parameter Index 1=true
    Write Parameter Index 2=true
   ) {enabled=false}
   &DynamicMaterialParameters(
    Index 0 Param 1=&Add_Float(
     A=$Particles.RandomFloat
     B=&FloatFromCurve(
      FloatCurve=DataInterface() {propertyValues {Curve {keys [
          {arriveTangent=2.7533, interpMode=RCIM_Cubic, leaveTangent=1.8958, tangentMode=RCTM_User, time=0, value=1}
          {arriveTangent=0.8292, interpMode=RCIM_Cubic, leaveTangent=0.8292, tangentMode=RCTM_User, time=0.321, value=1.3453}
          {arriveTangent=0.0529, interpMode=RCIM_Cubic, leaveTangent=0.2021, tangentMode=RCTM_User, time=1, value=1.5}
         ]}}}
      CurveIndex=&Add_Float(A=&Multiply_Float(A=$Particles.Initial.MaterialRandom, B=0.5), B=$Particles.NormalizedAge)
     )
    )
    Index 0 Param 2=&FloatFromCurve(FloatCurve=DataInterface() {propertyValues {Curve {keys [
         {arriveTangent=0.0453, interpMode=RCIM_Cubic, tangentMode=RCTM_Break, time=0, value=-1}
         {arriveTangent=0.0604, interpMode=RCIM_Cubic, leaveTangent=0.0604, tangentMode=RCTM_User, time=1, value=-1}
        ]}}})
    Index 0 Param 3=&FloatFromCurve(FloatCurve=DataInterface() {propertyValues {Curve {keys [
         {arriveTangent=-99.6442, interpMode=RCIM_Cubic, leaveTangent=-11.1131, tangentMode=RCTM_User, time=0, value=1}
         {arriveTangent=-0.2177, interpMode=RCIM_Cubic, leaveTangent=0.2723, tangentMode=RCTM_Break, time=0.25, value=0}
        ]}}})
    Index 0 Param 4=&FloatFromCurve(FloatCurve=DataInterface() {propertyValues {Curve {keys [<at 0.5 1>,<at 1 0>]}}})
    Index 1 Param 1=&FloatFromCurve(
     FloatCurve=DataInterface() {propertyValues {Curve {keys [<at 0 0>, {arriveTangent=12.7455, interpMode=RCIM_Cubic, leaveTangent=12.7455, tangentMode=RCTM_User, time=0.25, value=1}]}}}
     Scale Curve=&RandomRangeFloat(Minimum=1.5, Maximum=1.5)
    )
    Index 1 Param 2=&FloatFromCurve(
     FloatCurve=DataInterface() {propertyValues {Curve {keys [<at 0 0.2>, {leaveTangent=0.9066, time=1, value=2}]}}}
     Scale Curve=&RandomRangeFloat(Minimum=0.5, Maximum=0.6)
    )
    Index 1 Param 3=&FloatFromCurve(FloatCurve=DataInterface() {propertyValues {Curve {keys [
         {interpMode=RCIM_Cubic, leaveTangent=6.3917, tangentMode=RCTM_Break, time=0, value=0.2}
         {arriveTangent=1.7549, interpMode=RCIM_Cubic, leaveTangent=10.6767, tangentMode=RCTM_User, time=0.25, value=1}
        ]}}})
    Index 1 Param 4=&FloatFromCurve(
     FloatCurve=DataInterface() {propertyValues {Curve {keys [{arriveTangent=-0.136, interpMode=RCIM_Cubic, leaveTangent=-0.136, tangentMode=RCTM_User, time=0, value=1}, <at 1 0.95>]}}}
     Scale Curve=&RandomRangeFloat(Minimum=1.1, Maximum=1.4)
    )
    Index 2 Param 1=&FloatFromCurve(FloatCurve=DataInterface() {propertyValues {Curve {keys [
         {interpMode=RCIM_Cubic, leaveTangent=4.2485, tangentMode=RCTM_Break, time=0, value=0}
         {arriveTangent=1.0548, interpMode=RCIM_Cubic, leaveTangent=1.0548, tangentMode=RCTM_User, time=0.17, value=0.5715}
         <at 1 1>
        ]}}})
    Index 2 Param 2=&FloatFromCurve(
     FloatCurve=DataInterface() {propertyValues {Curve {keys [
         {arriveTangent=1.9056, interpMode=RCIM_Cubic, leaveTangent=1.9056, tangentMode=RCTM_User, time=0, value=0.2}
         {arriveTangent=0.0259, interpMode=RCIM_Cubic, leaveTangent=0.0259, tangentMode=RCTM_User, time=1, value=1.4}
        ]}}}
     Scale Curve=&RandomRangeFloat(Minimum=0.2, Maximum=0.7)
    )
    Index 2 Param 3=&FloatFromCurve(FloatCurve=DataInterface() {propertyValues {Curve {keys [
         {interpMode=RCIM_Cubic, leaveTangent=2.3064, tangentMode=RCTM_Break, time=0, value=1}
         {interpMode=RCIM_Cubic, leaveTangent=11.1233, tangentMode=RCTM_Break, time=1, value=2}
        ]}}})
    Index 2 Param 4=&FloatFromCurve(
     FloatCurve=DataInterface() {propertyValues {Curve {keys [
         {arriveTangent=-0.6867, interpMode=RCIM_Cubic, leaveTangent=-0.6867, tangentMode=RCTM_User, time=0, value=1}
         {interpMode=RCIM_Cubic, leaveTangent=-1.2956, tangentMode=RCTM_Break, time=1, value=0.75}
        ]}}}
     Scale Curve=1.6
    )
    Module.Param0WriteEnabled=true
    Module.Param0WriteEnabled1=true
    Module.Param0WriteEnabled2=true
    Module.Param1WriteEnabled=false
    Module.Param1WriteEnabled1=true
    Module.Param1WriteEnabled2=true
    Module.Param2WriteEnabled=true
    Module.Param2WriteEnabled1=true
    Module.Param2WriteEnabled2=true
    Module.Param3WriteEnabled=true
    Module.Param3WriteEnabled1=true
    Module.Param3WriteEnabled2=true
    Write Parameter Index 1=true
    Write Parameter Index 2=true
   ) {enabled=false, module=DynamicMaterialParameters001}
   &CameraOffset(Camera Offset Amount=15)
  ]
  rendererClasses [&NiagaraSpriteRendererProperties]
  renderers [
   {rendererClass=&NiagaraSpriteRendererProperties}
   {rendererClass=&NiagaraSpriteRendererProperties, rendererIndex=1}
  ]
 }
 Sparks {
  simTarget=GPUComputeSim
  emitterUpdateScript [
   &EmitterState(Life Cycle Mode=::NewEnumerator1, Loop Behavior=::NewEnumerator1, Loop Duration=0.25, UseLoopDelay=false)
   &SpawnBurst_Instantaneous(
    Spawn Count=&MakeIntFromFloat(Float=&Lerp_Float(A=10, B=30, Alpha=$System.NormalizedStageNum))
    Use Spawn Probability=false
   )
  ]
  particleSpawnScript [
   &InitializeParticle(
    Color=&LerpLinearColors(
     StartColor=<rgba 0.2588 0.1922 1 1>
     EndColor=&RandomRangeLinearColor(Minimum=<rgba 0 1 0.7667 1>, Maximum=<rgba 0.55 0 1 1>)
     LerpFactor=&UniformAOrBFloat(A=0, B=&UniformAOrBFloat(A=0))
    )
    Lifetime=&UniformRangedFloat(Minimum=0.05, Maximum=0.3)
    Module.Write Color=true
    Module.Write Lifetime=true
    Module.Write Mass=true
    Module.Write Material Random=true
    Module.Write Position=false
    Module.Write Scale=false
    Module.Write SpriteRotation=true
    Module.Write SpriteSize=true
    Position=&SimulationPosition()
    Sprite Rotation=&UniformRangedFloat(Minimum=0, Maximum=360)
    Sprite Size=&Vector2DFromFloat(Value=&RandomRangeFloat(Minimum=1, Maximum=3))
   )
   &SphereLocation(
    MaskSpawnGroup=false
    Non Uniform Scale=<1 1 0.1>
    Override Seed=false
    Sphere Radius=30
    Surface Only=false
    UseNonUniformScale=true
    UseOffset=false
    UseSphereOrientation=false
   )
   &AddVelocityFromPoint(
    Module.Use Falloff Distance=false
    Velocity Strength=&Multiply_Float(
     A=&RandomRangeFloat(Minimum=100, Maximum=200)
     B=&Lerp_Float(A=1, B=4, Alpha=$System.NormalizedStageNum)
    )
   )
   &AddVelocity(Velocity=&RandomRangeVector(Minimum=<0 0 100>, Maximum=<0 0 500>))
   /SetVariables(Transient.UpVelocity:Vector3f=$Particles.Velocity) {isSetParameters=true}
   /SetVariables() {isSetParameters=true}
   &AddVelocityFromPoint(
    Module.Use Falloff Distance=false
    Velocity Strength=&Multiply_Float(
     A=&RandomRangeFloat(Minimum=100, Maximum=500)
     B=&Lerp_Float(A=1, B=3, Alpha=$System.NormalizedStageNum)
    )
   ) {module=AddVelocityFromPoint001}
   /SetVariables(Transient.OutVelocity:Vector3f=$Particles.Velocity) {isSetParameters=true}
   /SetVariables(Particles.Velocity=&Lerp_Vector(A=$Transient.UpVelocity, B=$Transient.OutVelocity, Alpha=&UniformAOrBFloat(A=0))) {isSetParameters=true}
  ]
  particleUpdateScript [
   &ParticleState()
   &AccelerationForce(Acceleration=&RandomRangeVector(Minimum=<0 0 -200>, Maximum=<0 0 -100>))
   &Drag(Drag=3, Use Linear Drag=true, Use Rotational Drag=false)
   &SolveForcesAndVelocity(Clamp Velocity=false, Limit Acceleration=false)
   &ScaleColor(
    Scale Alpha=&Lerp_Float(A=1, B=0, Alpha=$Particles.NormalizedAge)
    Scale RGB=&VectorFromFloat(Value=&Lerp_Float(A=&RandomRangeFloat(Minimum=1000, Maximum=5000), B=1, Alpha=$Particles.NormalizedAge))
    ScaleA=true
    ScaleRGB=true
   )
   &SpriteRotationRate(Rotation Rate=&RandomRangeFloat(Minimum=-2, Maximum=2))
   &ScaleSpriteSizeBySpeed(Max Scale Factor=<1 20>, Min Scale Factor=<0.75 1>)
  ]
  rendererClasses [&NiagaraSpriteRendererProperties]
  renderers [{rendererClass=&NiagaraSpriteRendererProperties}]
 }
}
summary {
 systemName=NS_JumpBoots_OnJump
 emitters [
  {emitterName=SmokeGround, rendererClasses [&NiagaraSpriteRendererProperties]}
  {emitterName=SmokeBurst, rendererClasses [&NiagaraSpriteRendererProperties]}
  {bEnabled=false, emitterName=Empty, rendererClasses [&NiagaraSpriteRendererProperties]}
  {bEnabled=false, emitterName=Smoke_Sphere001, rendererClasses [&NiagaraSpriteRendererProperties]}
  {emitterName=Sparks, simTarget=GPUComputeSim, rendererClasses [&NiagaraSpriteRendererProperties]}
  {emitterName=Ashes, simTarget=GPUComputeSim, rendererClasses [&NiagaraSpriteRendererProperties]}
  {emitterName=SmokeLinger, rendererClasses [&NiagaraSpriteRendererProperties]}
  {emitterName=Beam, rendererClasses [&NiagaraMeshRendererProperties]}
 ]
 userVariables [User.StageNum {type=NiagaraInt32, defaultValue=2}]
}
