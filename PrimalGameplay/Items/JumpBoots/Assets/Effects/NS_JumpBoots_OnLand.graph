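# Niagara landing effect for the Jump Boots, switched by User.bValidLanding and scaled by User.StageNum.
# Valid landings burst smoke whose size, speed, lifetime and emissive vary with stage.
# Invalid landings instead briefly emit glowing, colored GPU sprites, gated by a stage check.
# Smoke randomly splits into short-lived puffs and flatter lingering particles, temporarily rescaling lingerers' NormalizedAge to speed up their erosion and flipbook material curves.
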NiagaraSystem /PrimalGameplay/Items/JumpBoots/Assets/Effects/NS_JumpBoots_OnLand
dependencies {
 usedDataInterfaces [&NiagaraDataInterfaceCurve, &NiagaraDataInterfaceColorCurve, &NiagaraDataInterfaceArrayFloat]
 usedDynamicInputs [&MakeFloatFromInt]
 usedModules [
  &self:SystemSpawnScript.SystemScriptSource.SystemScriptGraph.NiagaraNodeAssignment_0.SetVariables_#1
  &SystemState
 ]
}
emitters {
 Empty {
  bEnabled=false
  emitterUpdateScript [
   &EmitterState()
   &SpawnBurst_Instantaneous(Spawn Count=1, Use Spawn Probability=false)
  ]
  particleSpawnScript [&InitializeParticle(Sprite Size Mode=::NewEnumerator3, Uniform Sprite Size=100, UsePositionOffset=false, Write Material Random=true)]
  particleUpdateScript [&ParticleState()]
  rendererClasses [&NiagaraSpriteRendererProperties]
  renderers [{rendererClass=&NiagaraSpriteRendererProperties}]
 }
 FN_Sprite {
  simTarget=GPUComputeSim
  emitterUpdateScript [
   &EmitterState(Life Cycle Mode=::NewEnumerator1, Loop Behavior=::NewEnumerator1, Loop Duration=0.25, UseLoopDelay=false)
   &SpawnRate(
    Spawn Probability=0.5
    SpawnRate=&MakeCustomFloatFromBool(
     Bool=$User.bValidLanding
     True Float=0
     False Float=&MakeCustomFloatFromBool(
      Bool=&SetBoolByIntComparison(A=$User.StageNum, B=0, Comparison Type=::NewEnumerator4)
      True Float=200
      False Float=0
     )
    )
    Use Spawn Probability=true
   )
  ]
  particleSpawnScript [
   /Niagara/Modules/Spawn/Initialization/InitializeParticle(
    Color=&LerpLinearColors(
     StartColor=<rgba 0.2588 0.1922 1 1>
     EndColor=&RandomRangeLinearColor(Minimum=<rgba 0 1 0.7667 1>, Maximum=<rgba 0.55 0 1 1>)
     LerpFactor=&UniformAOrBFloat(A=0, B=&UniformAOrBFloat(A=0))
    )
    Lifetime=&UniformRangedFloat(Minimum=0.5, Maximum=1)
    Module.Write Color=true
    Module.Write Lifetime=true
    Module.Write Mass=true
    Module.Write Material Random=true
    Module.Write Position=false
    Module.Write Scale=false
    Module.Write SpriteRotation=true
    Module.Write SpriteSize=true
    Position=&SimulationPosition()
    Sprite Rotation=&UniformRangedFloat(Minimum=0, Maximum=360)
    Sprite Size=&Vector2DFromFloat(Value=&Lerp_Float(A=0.75, B=3, Alpha=&Power(a=&RandomRangeFloat(Minimum=0, Maximum=1), b=2)))
   )
   &SphereLocation(
    MaskSpawnGroup=false
    Non Uniform Scale=<0.5 0.5 1>
    Offset=<0 0 25>
    Offset Coordinate Space:ENiagaraCoordinateSpace=::Simulation
    Override Seed=false
    Sphere Radius=50
    Surface Only=false
    UseNonUniformScale=true
    UseOffset=true
    UseSphereOrientation=false
   )
   &AddVelocity(Velocity=&RandomRangeVector(Minimum=<-50 -50 -50>, Maximum=<50 50 50>))
  ]
  particleUpdateScript [
   &ParticleState()
   &AccelerationForce(Acceleration=&RandomRangeVector(Minimum=<0 0 -200>, Maximum=<0 0 -100>))
   &Drag(Drag=0.75, Use Linear Drag=true, Use Rotational Drag=false)
   &SolveForcesAndVelocity(Clamp Velocity=false, Limit Acceleration=false)
   &ScaleColor(
    Scale RGB=&VectorFromFloat(Value=&Lerp_Float(A=6000, B=0, Alpha=$Particles.NormalizedAge))
    ScaleA=false
    ScaleRGB=true
   )
   &SpriteRotationRate(Rotation Rate=&RandomRangeFloat(Minimum=-2, Maximum=2))
  ]
  rendererClasses [&NiagaraSpriteRendererProperties]
  renderers [{rendererClass=&NiagaraSpriteRendererProperties}]
 }
 Smoke {
  emitterUpdateScript [
   &EmitterState()
   &SpawnBurst_Instantaneous(
    Spawn Count=&MakeCustomIntFromBool(Bool=$User.bValidLanding, True INT=40, False INT=0)
    Spawn Time=0.1
    Use Spawn Probability=false
   )
  ]
  particleSpawnScript [
   /SetVariables(
    Particles.InitSpriteRotRate=&UniformRangedFloat(Minimum=-3, Maximum=3)
    Particles.IsLinger:NiagaraBool=&Boolean_AndOperation(Bool1=&RandomBool(), Bool2=&RandomBool())
    Transient.ScalePicker:NiagaraFloat=&UniformRangedFloat(Minimum=0, Maximum=1)
   ) {isSetParameters=true}
   &InitializeParticle(
    Color=<rgba 0.5 0.5 0.5 1>
    Lifetime=&MakeCustomFloatFromBool(
     Bool=$Particles.IsLinger
     True Float=&RandomRangeFloat(Minimum=0.75, Maximum=2)
     False Float=&Lerp_Float(
      A=&Lerp_Float(A=0.25, B=0.4, Alpha=$Transient.ScalePicker)
      B=&Lerp_Float(A=0.1, B=0.4, Alpha=$Transient.ScalePicker)
      Alpha=$System.NormalizedStageNum
     )
    )
    Module.Write Color=true
    Module.Write Lifetime=true
    Module.Write Mass=true
    Module.Write Material Random=true
    Module.Write Position=true
    Module.Write Scale=false
    Module.Write SpriteRotation=true
    Module.Write SpriteSize=true
    Sprite Rotation=&RandomRangeFloat(Minimum=-360, Maximum=360)
    Sprite Size=&Vector2DFromFloat(Value=&MakeCustomFloatFromBool(
      Bool=$Particles.IsLinger
      True Float=&Multiply_Float(
       A=&Lerp_Float(A=150, B=300, Alpha=&UniformRangedFloat(Minimum=0, Maximum=1))
       B=&Lerp_Float(A=0.6, B=1, Alpha=$System.NormalizedStageNum)
      )
      False Float=&Multiply_Float(
       A=&Lerp_Float(
        A=75
        B=150
        Alpha=&ClampFloat(Float=&Add_Float(A=$Transient.ScalePicker, B=&RandomRangeFloat(Minimum=-0.2, Maximum=0.3)))
       )
       B=&Lerp_Float(A=0.6, B=1, Alpha=$System.NormalizedStageNum)
      )
     ))
   )
   &SphereLocation(
    MaskSpawnGroup=false
    Non Uniform Scale=<1 1 0.75>
    Override Seed=true
    Random Seed=20
    Sphere Coordinate Space=::World
    Sphere Radius=&Lerp_Float(A=10, B=50, Alpha=&Lerp_Float(A=0.5, B=1, Alpha=$System.NormalizedStageNum))
    Surface Only=false
    UseNonUniformScale=true
    UseOffset=false
    UseSphereOrientation=false
   )
   &AddVelocityFromPoint(
    Module.Use Falloff Distance=false
    Velocity Strength=&Multiply_Float(
     A=&MakeCustomFloatFromBool(
      Bool=$Particles.IsLinger
      True Float=&RandomRangeFloat(Minimum=200, Maximum=500)
      False Float=&RandomRangeFloat(Minimum=100, Maximum=300)
     )
     B=&Lerp_Float(A=0.7, B=2, Alpha=$System.NormalizedStageNum)
    )
   )
   /SetVariables(Particles.Velocity=&MakeCustomVectorFromBool(
     Bool=$Particles.IsLinger
     True Vector=&Multiply_Vector(A=$Particles.Velocity, B=<1 1 0.25>)
     False Vector=&Multiply_Vector(
      A=&MakeVector(
       X=&MakeFloatFromVector(Vector=$Particles.Velocity, Channel=::NewEnumerator0)
       Y=&MakeFloatFromVector(Vector=$Particles.Velocity, Channel=::NewEnumerator1)
       Z=&Abs_Float(Float:NiagaraFloat=&MakeFloatFromVector(Vector=$Particles.Velocity, Channel=::NewEnumerator2))
      )
      B=<1 1 1.5>
     )
    )) {isSetParameters=true}
  ]
  particleUpdateScript [
   &ParticleState()
   &SpriteRotationRate(Rotation Rate=&Multiply_Float(
     A=$Particles.InitSpriteRotRate
     B=&FloatFromCurve(
      FloatCurve=DataInterface() {propertyValues {Curve {keys [
          {arriveTangent=-4.0107, interpMode=RCIM_Cubic, leaveTangent=-4.0107, tangentMode=RCTM_User, time=-0.0046, value=1.0113}
          {arriveTangent=-0.2496, interpMode=RCIM_Cubic, leaveTangent=-0.2496, tangentMode=RCTM_User, time=0.5, value=0.16}
         ]}}}
      CurveIndex=$Emitter.Age
      Scale Curve=150
     )
    ))
   /Niagara/Modules/Update/Forces/CurlNoiseForce(Noise Frequency=0.1, Noise Strength=800, Pan Noise=true, Pan Noise Field=<0 0 1>)
   &Drag(
    Drag=&MakeCustomFloatFromBool(Bool=$Particles.IsLinger, True Float=8, False Float=2)
    Use Linear Drag=true
    Use Rotational Drag=false
   )
   &SolveForcesAndVelocity(Clamp Velocity=false, Limit Acceleration=false)
   &ScaleSpriteSize(Scale Factor=&Vector2DFromFloat(Value=&FloatFromCurve(FloatCurve=DataInterface() {propertyValues {Curve {keys [
          {arriveTangent=2.8475, interpMode=RCIM_Cubic, leaveTangent=2.3251, tangentMode=RCTM_User, time=0, value=0.15}
          {arriveTangent=1.1483, interpMode=RCIM_Cubic, leaveTangent=1.1483, tangentMode=RCTM_User, time=0.2651, value=0.6514}
          {arriveTangent=0.2754, interpMode=RCIM_Cubic, leaveTangent=0.2754, tangentMode=RCTM_User, time=1.0463, value=1}
         ]}}})))
   &ScaleColor(
    Scale Alpha=&MakeCustomFloatFromBool(
     Bool=$Particles.IsLinger
     True Float=&FloatFromCurve(
      FloatCurve=DataInterface() {propertyValues {Curve {keys [
          {arriveTangent=-7.3919, interpMode=RCIM_Cubic, leaveTangent=-7.3919, tangentMode=RCTM_User, time=0, value=1}
          {arriveTangent=-0.6256, interpMode=RCIM_Cubic, leaveTangent=-0.6256, tangentMode=RCTM_User, time=0.1777, value=0.2782}
          <at 1 0>
         ]}}}
      Scale Curve=&Lerp_Float(A=0.5, B=1, Alpha=$System.NormalizedStageNum)
     )
     False Float=&FloatFromCurve(
      FloatCurve=DataInterface() {propertyValues {Curve {keys [
          {arriveTangent=0.1469, interpMode=RCIM_Cubic, leaveTangent=0.1469, tangentMode=RCTM_User, time=0.1213, value=0.984}
          {arriveTangent=-0.6758, interpMode=RCIM_Cubic, leaveTangent=-0.6758, tangentMode=RCTM_User, time=0.3002, value=0.1865}
          <at 1 0>
         ]}}}
      Scale Curve=&Lerp_Float(A=0.75, B=1, Alpha=$System.NormalizedStageNum)
     )
    )
    Scale RGB=&MakeVectorFromLinearColorRGB(Color=&ColorFromCurve(ColorCurve=DataInterface() {propertyValues {
        AlphaCurve {keys [<at 0 2 RCIM_Cubic>]}
        BlueCurve {keys [<at 0 0.3385 RCIM_Cubic>, {arriveTangent=-0.131, interpMode=RCIM_Cubic, leaveTangent=-0.131, time=0.2245, value=0.5}, <at 1.0334 0.2031>]}
        GreenCurve {keys [<at 0 0.2204 RCIM_Cubic>, {arriveTangent=-0.0167, interpMode=RCIM_Cubic, leaveTangent=-0.0167, time=0.2245, value=0.375}, <at 1.0334 0.2031>]}
        RedCurve {keys [<at 0 0.269 RCIM_Cubic>, {arriveTangent=-0.0637, interpMode=RCIM_Cubic, leaveTangent=-0.0637, time=0.2245, value=0.3958}, <at 1.0334 0.2031>]}
       }}))
    ScaleA=true
    ScaleRGB=true
   )
   /SetVariables(Particles.NormalizedAge:NiagaraFloat=&Multiply_Float(
     A=$Particles.NormalizedAge
     B=&MakeCustomFloatFromBool(Bool=$Particles.IsLinger, True Float=8, False Float=1)
    )) {isSetParameters=true}
   &self:Smoke_V2_Material_Dynamics_01(
    "Apply Noise (DissolveLerpNoise)"=1
    "Apply Noise (LerpNoise)"=1
    "Erode Alpha (DissolveNoiseErosionPower)":NiagaraFloat=&RandomRangeFloat(Minimum=0.5, Maximum=1)
    "Erode Emissive (NoiseErosionPower)":NiagaraFloat=&FloatFromCurve(FloatCurve=DataInterface() {propertyValues {Curve {keys [
         {arriveTangent=2.8986, interpMode=RCIM_Cubic, leaveTangent=2.8986, tangentMode=RCTM_User, time=0, value=0}
         {arriveTangent=5.2006, interpMode=RCIM_Cubic, leaveTangent=5.2006, tangentMode=RCTM_User, time=1, value=2}
        ]}}})
    "Intensity Modifier (Emissive)":NiagaraFloat=&FloatFromCurve(
     FloatCurve=DataInterface() {propertyValues {Curve {keys [{interpMode=RCIM_Cubic, tangentMode=RCTM_User, time=0, value=1}, <at 1 0>]}}}
     Scale Curve=&SelectFloatFromArray(Direct Array Index=$User.StageNum, Float Selection Array=DataInterface() {propertyValues {FloatData [0.05,0.1,0.5]}})
    )
    "Noise Offset X (DissolveNoiseOffsetX)":NiagaraFloat=&RandomRangeFloat(Minimum=0, Maximum=1)
    "Noise Offset X (NoiseOffsetX)":NiagaraFloat=&Add_Float(
     A=&RandomRangeFloat(Minimum=0, Maximum=1)
     B=&FloatFromCurve(
      FloatCurve=DataInterface() {propertyValues {Curve {keys [
          {arriveTangent=1.5771, interpMode=RCIM_Cubic, leaveTangent=1.223, tangentMode=RCTM_User, time=0, value=0}
          {arriveTangent=0.5984, interpMode=RCIM_Cubic, leaveTangent=0.5984, tangentMode=RCTM_User, time=0.321, value=0.3333}
          {arriveTangent=0.0529, interpMode=RCIM_Cubic, leaveTangent=0.2021, tangentMode=RCTM_User, time=1, value=0.5}
         ]}}}
      Scale Curve=&RandomRangeFloat(Minimum=0.25, Maximum=1)
     )
    )
    "Normalized Frame Number (Animation)":NiagaraFloat=&Add_Float(
     A=&RandomRangeFloat(Minimum=0, Maximum=1)
     B=&FloatFromCurve(FloatCurve=DataInterface() {propertyValues {Curve {keys [
          {arriveTangent=2.7533, interpMode=RCIM_Cubic, leaveTangent=1.8958, tangentMode=RCTM_User, time=0, value=1}
          {arriveTangent=0.8292, interpMode=RCIM_Cubic, leaveTangent=0.8292, tangentMode=RCTM_User, time=0.321, value=1.3453}
          {arriveTangent=0.0529, interpMode=RCIM_Cubic, leaveTangent=0.2021, tangentMode=RCTM_User, time=1, value=1.5}
         ]}}})
    )
    "Sphere Normal (NormalMultiplier)":NiagaraFloat=&FloatFromCurve(FloatCurve=DataInterface() {propertyValues {Curve {keys [<at 0.5 1>,<at 1 0>]}}})
    "Uniform Scale Noise (DissolveNoiseScale)":NiagaraFloat=&RandomRangeFloat(Minimum=1, Maximum=3)
    "Uniform Scale Noise (NoiseScale)":NiagaraFloat=&RandomRangeFloat(Minimum=1.1, Maximum=1.4)
   ) {module=Smoke_V2_Material_Dynamics_01}
   /SetVariables(Particles.NormalizedAge:NiagaraFloat=&Multiply_Float(
     A=$Particles.NormalizedAge
     B=&MakeCustomFloatFromBool(Bool=$Particles.IsLinger, True Float=0.125, False Float=1)
    )) {isSetParameters=true}
   &CameraOffset(Camera Offset Amount=15)
  ]
  rendererClasses [&NiagaraSpriteRendererProperties]
  renderers [
   {rendererClass=&NiagaraSpriteRendererProperties}
   {rendererClass=&NiagaraSpriteRendererProperties, rendererIndex=1}
  ]
 }
}
summary {
 systemName=NS_JumpBoots_OnLand
 emitters [
  {bEnabled=false, emitterName=Empty, rendererClasses [&NiagaraSpriteRendererProperties]}
  {emitterName=FN_Sprite, simTarget=GPUComputeSim, rendererClasses [&NiagaraSpriteRendererProperties]}
  {emitterName=Smoke, rendererClasses [&NiagaraSpriteRendererProperties]}
 ]
 userVariables [
  User.bValidLanding {type=NiagaraBool, defaultValue=true}
  User.StageNum {type=NiagaraInt32}
 ]
}
