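# Master material for the Tower jump boots, layering optional skin effects and anodized sheen between the character PreFX and PostFX functions.
# A Blueprint-supplied start time drives a decaying emissive pulse, with color and duration chosen by StageNum or IsLanding.
# The pulse dims base emissive and lights masked, Fresnel-shaped bands scrolling along pre-skinned height, with direction set by IsLanding.
# A time-of-day multiplier adjusts pulse brightness at night.
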Material /PrimalGameplay/Items/JumpBoots/Assets/Materials/M_Tower_Boots_Master
graph {
 properties {
  bUseMaterialAttributes=true
  bUsedWithMeshDeformer=true
  bUsedWithSkeletalMesh=true
  pixelDepthOffsetMode=PDOM_Legacy
  shadingModel=MSM_FromMaterialExpression
  subsurfaceProfile=&SS_HeroSkin_02
  usedShadingModels=Default Lit | Clear Coat | Subsurface Profile
 }
 nodes [
  n1 = MaterialFunctionCall() {materialFunction=&BaseFN_CharacterFunction_PreFX}
  n2 = MaterialFunctionCall!(In=TextureObjectParameter() {group=Skin FX, parameterName=SkinFX_Mask}) {materialFunction=/Game/Weapons/WeaponSkins/MaterialFunction/MF_GlitchSubfunction}
  n3 = StaticSwitchParameter(True=@n2.Glitched, False=@n2.Glitched) {group=Skin FX, parameterName=UseGlitch}
  n4 = StaticSwitchParameter(
   True=MaterialFunctionCall!(Alpha=ChannelMaskParameter(None=@n3) {group=Skin FX, parameterName=Silk Channel}.R, Material=@n1.Result) {materialFunction=/Game/Weapons/WeaponSkins/MaterialFunction/MF_Silk}.Result
   False=@n1.Result
  ) {group=Skin FX, parameterName=Use Silk}
  n5 = StaticSwitchParameter(
   True=MaterialFunctionCall!(Alpha=StaticComponentMaskParameter(None=@n3) {defaultR=true, group=Skin FX, parameterName=Clear Coat Channel}, In=@n4) {materialFunction=/Game/Weapons/WeaponSkins/MaterialFunction/Clearcoat}.Result
   False=@n4
  ) {group=Skin FX, parameterName=Use Clear Coat}
  n6 = StaticSwitchParameter(
   True=MaterialFunctionCall!(Alpha=StaticComponentMaskParameter(None=@n3) {defaultR=true, group=Skin FX, parameterName=ThinFilm_Channel}, Material Attributes=@n5) {materialFunction=/Game/Weapons/WeaponSkins/MaterialFunction/ThinFilmFunction}.Result
   False=@n5
  ) {group=Skin FX, parameterName=Use Thin Film}
  n7 = MaterialFunctionCall(Mask=@n3, Materials=@n6) {materialFunction=&MF_MetalsLUT}
  n8 = GetMaterialAttributes(None=MaterialFunctionCall!(In=StaticSwitchParameter(True=@n7.Result, False=@n6) {group=Skin FX, parameterName=UseMetalLUT}) {materialFunction=/Game/Weapons/WeaponSkins/MaterialFunction/MF_EmissiveFresnel}.Result) {attributeGetTypes [#1, #2]}
  n9 = Time()
  n10 = ScalarParameter() {defaultValue=100000000, group=TowerBootsPulsate, parameterName=GameTimeFromBP}
  n11 = ScalarParameter() {parameterName=StageNum}
  n12 = Constant() {r=1}
  n13 = ScalarParameter() {defaultValue=0.33, group=TowerBootsPulsate, parameterName=Duration3}
  n14 = ScalarParameter() {defaultValue=0.33, group=TowerBootsPulsate, parameterName=Duration2}
  n15 = ScalarParameter() {defaultValue=0.33, group=TowerBootsPulsate, parameterName=Duration}
  n16 = LinearInterpolate(
   A=If(A=@n11, B=Constant(), "A > B"=If(A=@n11, B=@n12, "A > B"=@n13, "A == B"=@n14, "A < B"=@n15), "A == B"=@n15, "A < B"=@n15)
   B=@n15
   Alpha=ScalarParameter() {parameterName=IsLanding}
  )
  MaterialExpressionSaturate_0 = Saturate(None=Divide(A=Subtract(A=@n9, B=@n10), B=Multiply(A=@n16) {constB=0.5}))
  n17 = Custom!(InputFunction=@MaterialExpressionSaturate_0) {
   code="float2 result = (2 * InputFunction) -2;\r\nresult  = result  * result  * result ;\r\nreturn -1.181 * InputFunction * result;"
   description=FastAttackSlowDecay
   outputType=CMOT_Float2
   inputs [{
     inputName=InputFunction
     input {expression=&self:MaterialExpressionSaturate_0}
    }]
  }
  n18 = ScalarParameter() {defaultValue=3, group=TowerBootsPulsate, parameterName=PulsesCount}
  n19 = ConstantBiasScale(None=Sine(None=Multiply(A=@MaterialExpressionSaturate_0, B=@n18) {constB=3}))
  n20 = Multiply(
   A=ComponentMask(None=@n17) {r=true}
   B=LinearInterpolate(
    A=ScalarParameter() {group=TowerBootsPulsate, parameterName=MinPulse}
    B=ScalarParameter() {defaultValue=1, group=TowerBootsPulsate, parameterName=MaxPulse}
    Alpha=Saturate(None=@n19)
   )
  )
  n21 = Multiply(A=@n8.Emissive Color, B=OneMinus(None=Saturate(None=@n20)))
  n22 = ScalarParameter() {parameterName=StageNum}
  n23 = VectorParameter() {group=TowerBootsPulsate, parameterName=PulseColor, defaultValue=<rgba 0 3 1.292 0>}
  n24 = If(
   A=@n22
   B=Constant()
   "A > B"=If(
    A=@n22
    B=Constant() {r=1}
    "A > B"=VectorParameter() {group=TowerBootsPulsate, parameterName=PulseColor3, defaultValue=<rgba 0 3 1.292 0>}.RGB
    "A == B"=VectorParameter() {group=TowerBootsPulsate, parameterName=PulseColor2, defaultValue=<rgba 0 3 1.292 0>}.RGB
    "A < B"=@n23.RGB
   )
   "A == B"=@n23.RGB
   "A < B"=@n23.RGB
  )
  n25 = ScalarParameter() {parameterName=IsLanding}
  n26 = Multiply(
   A=@n20
   B=Multiply(
    A=TextureSampleParameter2D() {parameterName=PulsateEmissiveMask, texture=/PrimalGameplay/Items/JumpBoots/Assets/Textures/T_Tower_Boots_E}.G
    B=LinearInterpolate(A=@n24, B=@n23.RGB, Alpha=@n25)
   )
  )
  n27 = PreSkinnedPosition()
  n28 = PreSkinnedLocalBounds()
  n29 = VertexInterpolator(VS=Divide(A=Subtract(A=@n27, B=@n28.Min), B=@n28.Extents))
  n30 = Multiply(
   A=Saturate(None=ComponentMask(None=@n29.PS) {b=true})
   B=ScalarParameter() {defaultValue=1, group=TowerBootsPulsate, parameterName=Lights_count}
  ) {constB=10}
  n31 = Subtract(A=Multiply(A=ScalarParameter() {parameterName=IsLanding}) {constB=2})
  n32 = Multiply(A=ScalarParameter() {defaultValue=3, group=TowerBootsPulsate, parameterName=Lights_Speed}, B=@n31)
  n33 = SphereMask(
   A=Frac(None=Panner(Coordinate=@n30, Speed=@n32) {speedY=1})
   B=Constant() {r=0.5}
   Radius=ScalarParameter() {defaultValue=0.6, group=TowerBootsPulsate, parameterName=Lights_radius}
   Hardness=ScalarParameter() {group=TowerBootsPulsate, parameterName=Lights_hardness}
  )
  n34 = PixelNormalWS()
  n35 = StaticBool() {value=true}
  n36 = ScalarParameter() {defaultValue=1, group=TowerBootsPulsate, parameterName=fresnelPower}
  n37 = Constant() {r=1}
  n38 = Multiply(
   A=Multiply(
    A=Multiply(A=@n26, B=@n33)
    B=MaterialFunctionCall(Normal Vector=@n34, Invert Fresnel=@n35, Power=@n36, Clamp Fresnel Dot Product=@n35) {materialFunction=&Fresnel_Function}.Result
   )
   B=MaterialFunctionCall(
    Value_Morning=@n37
    Value_Day=@n37
    Value_Evening=@n37
    Value_Night=ScalarParameter() {defaultValue=0.5, group=TowerBootsPulsate, parameterName=NightMultiplier}
   ) {materialFunction=&Scalar_Time_of_Day}.Result
  )
  n39 = Constant() {r=1}
  n40 = Constant() {r=3}
  n41 = PixelNormalWS()
  n42 = Multiply(
   A=TextureSample(UVs=Fresnel(ExponentIn=@n39, BaseReflectFractionIn=@n40, Normal=@n41) {baseReflectFraction=0.5, exponent=3}) {texture=/Game/Global/Textures/Default/Masks_and_Gradients/3pxRGBImage}.RGB
   B=VectorParameter() {group=TowerSentinel_Metal, parameterName=AnodizationFilter, defaultValue=<rgba 1 0 1 0>}.RGB
  )
  n43 = Constant() {r=4}
  n44 = AtmosphericLightVector()
  n45 = ReflectionVectorWS()
  n46 = Multiply(
   A=Power(Base=DotProduct(A=@n44, B=@n45), Exp=ScalarParameter() {defaultValue=55, group=TowerSentinel_Metal, parameterName=Anodized_pow})
   B=Constant() {r=512}
  )
  n47 = TextureSampleParameter2D() {
   group=TowerSentinel_Metal
   parameterName=TowerSentinelMetals
   samplerSource=SSM_Wrap_WorldGroupSettings
   samplerType=SAMPLERTYPE_Masks
   texture=&T_WhiteMask
  }
  n48 = Multiply(
   A=LinearInterpolate(A=@n42, B=Multiply(A=@n43, B=@n42), Alpha=Saturate(None=@n46))
   B=LinearInterpolate(
    A=ScalarParameter() {group=TowerSentinel_Metal, parameterName=AnodizedEffectIntensity}
    B=ScalarParameter() {group=TowerSentinel_Metal, parameterName=AnodizedEffectIntensity2}
    Alpha=StaticComponentMaskParameter(None=@n47.RGB) {defaultR=true, group=TowerSentinel_Metal, parameterName=TowerSentinelMetals_Channel}
   )
  ) {constB=0.1}
  n49 = TextureSampleParameter2D() {
   group=TowerSentinel_Metal
   parameterName=AnodizationEffectMask
   samplerSource=SSM_Wrap_WorldGroupSettings
   samplerType=SAMPLERTYPE_Masks
   texture=&T_WhiteMask
  }
  n50 = Constant()
  n51 = StaticSwitchParameter(True=ShadingPathSwitch(Default=Multiply(A=@n48, B=@n49.R), Forward=@n50, Mobile=@n50), False=@n50) {group=TowerSentinel, parameterName=UseAnodizedEffect}
  MaterialFunctionCall(
   AO=@n1.AO
   Distance Blend=@n1.AO
   Fresnel=@n1.AO
   Gmap=@n1.AO
   Pre Skinned Local Position=@n1.AO
   Material=SetMaterialAttributes!(MaterialAttributes=@n8.MaterialAttributes, Emissive Color=Add(A=@n21, B=@n38), Base Color=Add(A=@n8.Base Color, B=@n51)) {attributeSetTypes [#1, #2]}
  ) {materialFunction=&BaseFN_CharacterFunction_PostFX}
  Multiply(A=@n31) {constB=-1}
  ScalarParameter() {parameterName=IsLanding}
 ]
}
