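# Animation Blueprint for the Tower jump boots mesh, posing it to follow the wearer's legs and feet.
# The parent mesh's copied pose takes IK-retargeted ball bones, then look-ats and rotation constraints adjust boot bones.
# Each update converts foot-versus-calf twist into per-foot alphas, smoothly weighting the boot_tp_01 constraints to the feet.
# The right side's first look-at targets its own bone (boot_tp_02_r), unlike the left.
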AnimBlueprint /PrimalGameplay/Items/JumpBoots/Assets/Meshes/Tower_Boots_AnimBP
target_skeleton=&/Tower_Boots_Skeleton
graphs [
 AnimGraph {nodes [
   n1 = Copy Pose From Mesh() {bUseAttachedParent=true}
   n2 = Retarget Pose From Mesh!() -> (Pose:Pose Link Structure) {is=title AnimGraphNode_RetargetPoseFromMesh, iKRetargeterAsset=&/RTG_Tower_Boots}
   n3 = Look At - Bone: boot_02_l(
    ComponentPose=Local To Component(LocalPose=Layered blend per bone(BasePose=@n1, BlendPoses_0=@n2) {layerSetup [{branchFilters [{boneName=ball_l}, {boneName=ball_r}]}]})
    Alpha=1
   ) {
    alphaBoolBlend {blendOption=Linear}
    boneToModify {boneName=boot_02_l}
    lookAtLocation {x=0}
    lookAtTarget {boneReference {boneName=boot_tp_02_l}}
    lookAt_Axis {axis=<1 0>}
    lookUp_Axis {axis {x=0, z=1}}
   }
   n4 = Look At - Bone: boot_tp_02_l(ComponentPose=@n3, Alpha=1) {
    alphaBoolBlend {blendOption=Linear}
    boneToModify {boneName=boot_tp_02_l}
    lookAtLocation {x=0}
    lookAtTarget {boneReference {boneName=boot_tp_02_l_aim}}
    lookAt_Axis {axis=<1 0>}
    lookUp_Axis {axis {x=0, z=1}}
   }
   n5 = "Constraint - boot_01_l (R)"(ComponentPose=@n4, Alpha=0.5, ConstraintWeights_0=1) {
    alphaBoolBlend {blendOption=Linear}
    boneToModify {boneName=boot_01_l}
    constraintSetup [{transformType=Rotation, targetBone {boneName=foot_l}}]
    constraintWeights [1]
   }
   n6 = "Constraint - boot_tp_01_l (R)"(ComponentPose=@n5, Alpha=Get leftFootAlpha() -> (leftFootAlpha:double), ConstraintWeights_0=1) {
    alphaBoolBlend {blendOption=Linear}
    alphaScaleBiasClamp {interpSpeedDecreasing=3, interpSpeedIncreasing=3}
    boneToModify {boneName=boot_tp_01_l}
    constraintSetup [{transformType=Rotation, targetBone {boneName=foot_l}}]
    constraintWeights [1]
   }
   n7 = Look At - Bone: boot_tp_02_r(ComponentPose=@n6, Alpha=1) {
    alphaBoolBlend {blendOption=Linear}
    boneToModify {boneName=boot_tp_02_r}
    lookAtLocation {x=0}
    lookAtTarget {boneReference {boneName=boot_tp_02_r}}
    lookAt_Axis {axis=<1 0>}
    lookUp_Axis {axis {x=0, z=1}}
   }
   n8 = Look At - Bone: boot_tp_02_r(ComponentPose=@n7, Alpha=1) {
    alphaBoolBlend {blendOption=Linear}
    boneToModify {boneName=boot_tp_02_r}
    lookAtLocation {x=0}
    lookAtTarget {boneReference {boneName=boot_tp_02_r_aim}}
    lookAt_Axis {axis=<1 0>}
    lookUp_Axis {axis {x=0, z=1}}
   }
   n9 = "Constraint - boot_01_r (R)"(ComponentPose=@n8, Alpha=0.5, ConstraintWeights_0=1) {
    alphaBoolBlend {blendOption=Linear}
    boneToModify {boneName=boot_01_r}
    constraintSetup [{transformType=Rotation, targetBone {boneName=foot_r}}]
    constraintWeights [1]
   }
   n10 = "Constraint - boot_tp_01_r (R)"(ComponentPose=@n9, Alpha=Get rightFootAlpha() -> (rightFootAlpha:double), ConstraintWeights_0=1) {
    alphaBoolBlend {blendOption=Linear}
    alphaScaleBiasClamp {interpSpeedDecreasing=3, interpSpeedIncreasing=3}
    boneToModify {boneName=boot_tp_01_r}
    constraintSetup [{transformType=Rotation, targetBone {boneName=foot_r}}]
    constraintWeights [1]
   }
   Output Pose(Result=Component To Local(ComponentPose=@n10)) {layerGroup=None}
  ]}
 EventGraph {nodes [
   Event BlueprintUpdateAnimation()
   n1 = GetOwnerSkeletalMeshComponent()
   > Set leftFootAlpha(leftFootAlpha:double=GetAlphaFromReferenceTransformDelta(SkeletalMeshComponent=@n1, BoneName=foot_l, BaseName=calf_l, InputRange=(X=-30,Y=-20), OutputRange=(X=0.1,Y=0.5), TwistAxis="1,0,0")) -> (Output_Get:double)
   > Set rightFootAlpha(rightFootAlpha:double=GetAlphaFromReferenceTransformDelta(SkeletalMeshComponent=@n1, BoneName=foot_r, BaseName=calf_r, InputRange=(X=-30,Y=-20), OutputRange=(X=0.1,Y=0.5), TwistAxis="1,0,0")) -> (Output_Get:double)
   TryGetPawnOwner()
  ]}
]
variables [
 leftFootAlpha {type=number}
 rightFootAlpha {type=number}
]
